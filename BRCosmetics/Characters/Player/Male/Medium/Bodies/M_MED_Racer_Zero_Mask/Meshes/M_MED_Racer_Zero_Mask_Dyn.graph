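# Physics asset for the M_MED_Racer_Zero_Mask male cosmetic body, set up for dynamic motion of its long coat.
# Capsules cover pelvis, spine, thighs and calves, while capsules and boxes cover seven coat chains, the center back splitting into two tails.
# Each chain hangs from a free-rotating master body on spine_03, with angle-limited joints below.
# Free-rotation cross constraints tie neighboring panels together at intervals, leaving the front unlinked between sides.
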PhysicsAsset /BRCosmetics/Characters/Player/Male/Medium/Bodies/M_MED_Racer_Zero_Mask/Meshes/M_MED_Racer_Zero_Mask_Dyn
bodies {
 calf_l [{shapeType=Capsule, length=46.5884, radius=8.4266, center=<-26.3168 -2.7608 0.0003>, rotation=<rot 90 0 0>}]
 calf_r [{shapeType=Capsule, length=46.5884, radius=8.4266, center=<27.0703 2.7608 -0.0004>, rotation=<rot 90 0 0>}]
 dyn_coat_bk_01 [{shapeType=Box, extentX=10.3031, extentY=16.635, extentZ=1.8698, center=<4.5137 0.0004 0.0001>}]
 dyn_coat_bk_02 [{
   shapeType=Capsule
   length=25.5413
   radius=4.4587
   shapeName=dyn_coat_bk_l_02_capsule
   center=<8.054 0.0012 1.1495>
   rotation=<rot 0 0 89.9999>
  }]
 dyn_coat_bk_03 [{shapeType=Capsule, length=18.9791, radius=5.5209, shapeName=dyn_coat_bk_l_06_capsule, center=<6.3 0 1.1081>, rotation=<rot 0 0 89.9999>}]
 dyn_coat_bk_MSTR [{shapeType=Capsule, length=0.505, radius=0.505}]
 dyn_coat_bk_l_01 [{shapeType=Box, extentX=9.5798, extentY=16.635, extentZ=1.8698, center=<4.522 0.0004 0.0001>}]
 dyn_coat_bk_l_02 [{shapeType=Capsule, length=19.7213, radius=10.2787, shapeName=dyn_coat_bk_l_02_capsule, center=<8.0539 0.0011 10.7802>, rotation=<rot 0 0 90>}]
 dyn_coat_bk_l_03 [{shapeType=Capsule, length=22.5958, radius=9.0599, shapeName=dyn_coat_bk_l_06_capsule, center=<6.2997 -0.0001 10.1952>, rotation=<rot 0 0 90>}]
 dyn_coat_bk_l_04 [{shapeType=Capsule, length=22.5958, radius=9.0599, shapeName=dyn_coat_bk_l_06_capsule, center=<6.3 -0.0001 6.7728>, rotation=<rot 0 0 90>}]
 dyn_coat_bk_l_05 [{shapeType=Capsule, length=30, radius=7, shapeName=dyn_coat_bk_l_06_capsule, center=<6.3 -0.0001 6.7728>, rotation=<rot 0 0 90>}]
 dyn_coat_bk_l_06 [{shapeType=Box, extentX=16.0106, extentY=28.4046, extentZ=3.0378, shapeName=dyn_coat_bk_r_06_box, center=<7.3188 0.0001 0>}]
 dyn_coat_bk_l_MSTR [{shapeType=Capsule, length=0.505, radius=0.505}]
 dyn_coat_bk_r_01 [{shapeType=Box, extentX=9.5798, extentY=16.635, extentZ=1.8698, center=<4.522 0.0004 0.0001>}]
 dyn_coat_bk_r_02 [{shapeType=Capsule, length=19.7213, radius=10.2787, shapeName=dyn_coat_bk_l_02_capsule, center=<8.0539 0.0011 10.7802>, rotation=<rot 0 0 90>}]
 dyn_coat_bk_r_03 [{shapeType=Capsule, length=22.5958, radius=9.0599, shapeName=dyn_coat_bk_l_06_capsule, center=<6.2997 -0.0001 10.1952>, rotation=<rot 0 0 90>}]
 dyn_coat_bk_r_04 [{shapeType=Capsule, length=22.5958, radius=9.0599, shapeName=dyn_coat_bk_l_06_capsule, center=<6.3 -0.0001 6.7728>, rotation=<rot 0 0 90>}]
 dyn_coat_bk_r_05 [{shapeType=Capsule, length=30, radius=7, shapeName=dyn_coat_bk_l_06_capsule, center=<6.3 -0.0001 6.7728>, rotation=<rot 0 0 90>}]
 dyn_coat_bk_r_06 [{shapeType=Box, extentX=16.0106, extentY=28.4046, extentZ=3.0378, shapeName=dyn_coat_bk_r_06_box, center=<7.3188 0.0001 0>}]
 dyn_coat_bk_r_MSTR [{shapeType=Capsule, length=0.505, radius=0.505}]
 dyn_coat_bkend_l_01 [{shapeType=Capsule, length=7.2031, radius=8.4058, shapeName=dyn_coat_bk_l_06_capsule, center=<6.3 0 7.5798>, rotation=<rot 0 0 90>}]
 dyn_coat_bkend_l_02 [{shapeType=Capsule, length=21.811, radius=7, shapeName=dyn_coat_bk_l_06_capsule, center=<6.3 0 6.7728>, rotation=<rot 90 0 0>}]
 dyn_coat_bkend_l_03 [{shapeType=Box, extentX=16.0106, extentY=28.4046, extentZ=3.0378, shapeName=dyn_coat_bk_r_06_box, center=<7.3188 0.0001 0>}]
 dyn_coat_bkend_r_01 [{shapeType=Capsule, length=7.2031, radius=8.4058, shapeName=dyn_coat_bk_l_06_capsule, center=<6.3 0 7.5799>, rotation=<rot 0 0 90>}]
 dyn_coat_bkend_r_02 [{shapeType=Capsule, length=21.811, radius=7, shapeName=dyn_coat_bk_l_06_capsule, center=<6.3 0 6.7728>, rotation=<rot 90 0 0>}]
 dyn_coat_bkend_r_03 [{shapeType=Box, extentX=16.0106, extentY=28.4046, extentZ=3.0378, shapeName=dyn_coat_bk_r_06_box, center=<7.3188 0.0001 0>}]
 dyn_coat_fr_l_01 [{shapeType=Capsule, length=4.9641, radius=5.7047, shapeName=dyn_coat_bk_l_06_capsule, center=<6.3 -0.0001 6.7728>, rotation=<rot 0 0 90>}]
 dyn_coat_fr_l_02 [{shapeType=Capsule, length=8.1176, radius=5.7047, shapeName=dyn_coat_bk_l_06_capsule, center=<6.3 -0.0001 5.0779>, rotation=<rot 0 0 90>}]
 dyn_coat_fr_l_03 [{shapeType=Capsule, length=6.549, radius=9.0599, shapeName=dyn_coat_bk_l_06_capsule, center=<6.3 -0.0002 8.2083>, rotation=<rot 0 0 90>}]
 dyn_coat_fr_l_04 [{shapeType=Capsule, length=6.549, radius=9.0599, shapeName=dyn_coat_bk_l_06_capsule, center=<6.3 -0.0001 6.7728>, rotation=<rot 0 0 90>}]
 dyn_coat_fr_l_05 [{shapeType=Capsule, length=6.549, radius=9.0599, shapeName=dyn_coat_bk_l_06_capsule, center=<6.3 -0.0001 6.7728>, rotation=<rot 0 0 90>}]
 dyn_coat_fr_l_06 [{shapeType=Capsule, length=6.549, radius=9.0599, shapeName=dyn_coat_bk_l_06_capsule, center=<6.3 -0.0001 6.7728>, rotation=<rot 0 0 90>}]
 dyn_coat_fr_l_MSTR [{shapeType=Capsule, length=0.505, radius=0.505}]
 dyn_coat_fr_r_01 [{shapeType=Capsule, length=5, radius=6, shapeName=dyn_coat_bk_l_06_capsule, center=<5.2806 0.0002 4.7302>, rotation=<rot 0 0 89.9999>}]
 dyn_coat_fr_r_02 [{
   shapeType=Capsule
   length=13.0906
   radius=5.4625
   shapeName=dyn_coat_l_02_capsule
   center=<5.3535 -0.0004 3.8515>
   rotation=<rot 0 0 89.9999>
  }]
 dyn_coat_fr_r_03 [{shapeType=Capsule, length=6.549, radius=9.0599, shapeName=dyn_coat_bk_l_06_capsule, center=<6.3 0.0001 7.8107>, rotation=<rot 0 0 89.9999>}]
 dyn_coat_fr_r_04 [{shapeType=Capsule, length=6.549, radius=9.0599, shapeName=dyn_coat_bk_l_06_capsule, center=<6.3 -0.0001 6.7728>, rotation=<rot 0 0 90>}]
 dyn_coat_fr_r_05 [{shapeType=Capsule, length=6.549, radius=9.0599, shapeName=dyn_coat_bk_l_06_capsule, center=<6.3 -0.0001 6.7728>, rotation=<rot 0 0 90>}]
 dyn_coat_fr_r_06 [{shapeType=Capsule, length=6.549, radius=9.0599, shapeName=dyn_coat_bk_l_06_capsule, center=<6.3 -0.0001 6.7728>, rotation=<rot 0 0 90>}]
 dyn_coat_fr_r_MSTR [{shapeType=Capsule, length=0.505, radius=0.505}]
 dyn_coat_l_01 [{shapeType=Box, extentX=9.0085, extentY=16.635, extentZ=1.8698, center=<3.9758 0.0006 0>}]
 dyn_coat_l_02 [{shapeType=Capsule, length=26.0012, radius=5.4625, shapeName=dyn_coat_l_02_capsule, center=<3.5182 -0.0004 3.8515>, rotation=<rot 0 0 90>}]
 dyn_coat_l_03 [{shapeType=Capsule, length=19.4595, radius=9.0599, shapeName=dyn_coat_bk_l_06_capsule, center=<6.3 -0.0001 6.7728>, rotation=<rot 0 0 90>}]
 dyn_coat_l_04 [{shapeType=Capsule, length=19.4595, radius=9.0599, shapeName=dyn_coat_bk_l_06_capsule, center=<6.3 -0.0001 6.7728>, rotation=<rot 0 0 90>}]
 dyn_coat_l_05 [{shapeType=Capsule, length=19.4595, radius=9.0599, shapeName=dyn_coat_bk_l_06_capsule, center=<6.3 -0.0001 6.7728>, rotation=<rot 0 0 90>}]
 dyn_coat_l_06 [{shapeType=Box, extentX=16.0106, extentY=28.4046, extentZ=3.0378, shapeName=dyn_coat_bk_r_06_box, center=<7.3188 0.0001 0>}]
 dyn_coat_l_MSTR [{shapeType=Capsule, length=0.505, radius=0.505}]
 dyn_coat_r_01 [{shapeType=Box, extentX=9.0085, extentY=16.635, extentZ=1.8698, center=<3.9758 0.0006 0>}]
 dyn_coat_r_02 [{shapeType=Box, extentX=6.2726, extentY=16.635, extentZ=1.8698, center=<3.4585 0.0003 0>}]
 dyn_coat_r_03 [{shapeType=Capsule, length=19.4595, radius=9.0599, shapeName=dyn_coat_bk_l_06_capsule, center=<6.3 -0.0001 6.7728>, rotation=<rot 0 0 90>}]
 dyn_coat_r_04 [{shapeType=Capsule, length=19.4595, radius=9.0599, shapeName=dyn_coat_bk_l_06_capsule, center=<6.3 -0.0001 6.7728>, rotation=<rot 0 0 90>}]
 dyn_coat_r_05 [{shapeType=Capsule, length=19.4595, radius=9.0599, shapeName=dyn_coat_bk_l_06_capsule, center=<6.3 -0.0001 6.7728>, rotation=<rot 0 0 90>}]
 dyn_coat_r_06 [{shapeType=Box, extentX=16.0106, extentY=28.4046, extentZ=3.0378, shapeName=dyn_coat_bk_r_06_box, center=<7.3188 0.0001 0>}]
 dyn_coat_r_MSTR [{shapeType=Capsule, length=0.505, radius=0.505}]
 pelvis [{shapeType=Capsule, length=14.3465, radius=10.8041, center=<-7.1711 -1.5828 0>}]
 spine_01 [{shapeType=Capsule, length=9.6251, radius=11.3666, center=<-2.4193 -0.9098 0>}]
 spine_02 [{shapeType=Capsule, length=7.5659, radius=10.8041, center=<-2.4193 -1.1055 0>}]
 spine_03 [{shapeType=Capsule, length=9.0791, radius=9.5015, center=<-2.4192 -1.4331 0>}]
 thigh_l [{shapeType=Capsule, length=31.8916, radius=8.4771, center=<-19.3302 0.2527 -0.3704>, rotation=<rot 90 0 0>}]
 thigh_r [{shapeType=Capsule, length=31.8916, radius=8.4771, center=<19.3302 -0.2527 0.3704>, rotation=<rot 90 0 0>}]
}
constraints [
 {
  bone1Name=dyn_coat_bk_l_MSTR
  bone2Name=spine_03
  swing1LimitDegrees=45
  swing1Motion=Free
  swing2LimitDegrees=45
  swing2Motion=Free
  twistLimitDegrees=45
  twistMotion=Free
 }
 {bone1Name=dyn_coat_bk_l_01, bone2Name=dyn_coat_bk_l_MSTR, swing1LimitDegrees=45, swing2LimitDegrees=45, twistLimitDegrees=5}
 {bone1Name=dyn_coat_bk_l_02, bone2Name=dyn_coat_bk_l_01, swing1LimitDegrees=45, swing2LimitDegrees=45, twistLimitDegrees=5}
 {bone1Name=dyn_coat_bk_l_03, bone2Name=dyn_coat_bk_l_02, swing1LimitDegrees=45, swing2LimitDegrees=45, twistLimitDegrees=5}
 {bone1Name=dyn_coat_bk_l_04, bone2Name=dyn_coat_bk_l_03, swing1LimitDegrees=45, swing2LimitDegrees=45, twistLimitDegrees=5}
 {bone1Name=dyn_coat_bk_l_05, bone2Name=dyn_coat_bk_l_04, swing1LimitDegrees=45, swing2LimitDegrees=45, twistLimitDegrees=5}
 {bone1Name=dyn_coat_bk_l_06, bone2Name=dyn_coat_bk_l_05, swing1LimitDegrees=45, swing2LimitDegrees=45, twistLimitDegrees=5}
 {
  bone1Name=dyn_coat_l_MSTR
  bone2Name=spine_03
  swing1LimitDegrees=45
  swing1Motion=Free
  swing2LimitDegrees=45
  swing2Motion=Free
  twistLimitDegrees=45
  twistMotion=Free
 }
 {bone1Name=dyn_coat_l_01, bone2Name=dyn_coat_l_MSTR, swing1LimitDegrees=45, swing2LimitDegrees=45, twistLimitDegrees=5}
 {bone1Name=dyn_coat_l_02, bone2Name=dyn_coat_l_01, swing1LimitDegrees=45, swing2LimitDegrees=45, twistLimitDegrees=5}
 {bone1Name=dyn_coat_l_03, bone2Name=dyn_coat_l_02, swing1LimitDegrees=45, swing2LimitDegrees=45, twistLimitDegrees=5}
 {bone1Name=dyn_coat_l_04, bone2Name=dyn_coat_l_03, swing1LimitDegrees=45, swing2LimitDegrees=45, twistLimitDegrees=5}
 {bone1Name=dyn_coat_l_05, bone2Name=dyn_coat_l_04, swing1LimitDegrees=45, swing2LimitDegrees=45, twistLimitDegrees=5}
 {bone1Name=dyn_coat_l_06, bone2Name=dyn_coat_l_05, swing1LimitDegrees=45, swing2LimitDegrees=45, twistLimitDegrees=5}
 {
  bone1Name=dyn_coat_fr_l_MSTR
  bone2Name=spine_03
  swing1LimitDegrees=45
  swing1Motion=Free
  swing2LimitDegrees=45
  swing2Motion=Free
  twistLimitDegrees=45
  twistMotion=Free
 }
 {bone1Name=dyn_coat_fr_l_01, bone2Name=dyn_coat_fr_l_MSTR, swing1LimitDegrees=45, swing2LimitDegrees=45, twistLimitDegrees=5}
 {bone1Name=dyn_coat_fr_l_02, bone2Name=dyn_coat_fr_l_01, swing1LimitDegrees=45, swing2LimitDegrees=45, twistLimitDegrees=5}
 {bone1Name=dyn_coat_fr_l_03, bone2Name=dyn_coat_fr_l_02, swing1LimitDegrees=45, swing2LimitDegrees=45, twistLimitDegrees=5}
 {bone1Name=dyn_coat_fr_l_04, bone2Name=dyn_coat_fr_l_03, swing1LimitDegrees=45, swing2LimitDegrees=45, twistLimitDegrees=5}
 {bone1Name=dyn_coat_fr_l_05, bone2Name=dyn_coat_fr_l_04, swing1LimitDegrees=45, swing2LimitDegrees=45, twistLimitDegrees=5}
 {bone1Name=dyn_coat_fr_l_06, bone2Name=dyn_coat_fr_l_05, swing1LimitDegrees=45, swing2LimitDegrees=45, twistLimitDegrees=5}
 {
  bone1Name=dyn_coat_fr_r_MSTR
  bone2Name=spine_03
  swing1LimitDegrees=45
  swing1Motion=Free
  swing2LimitDegrees=45
  swing2Motion=Free
  twistLimitDegrees=45
  twistMotion=Free
 }
 {bone1Name=dyn_coat_fr_r_01, bone2Name=dyn_coat_fr_r_MSTR, swing1LimitDegrees=45, swing2LimitDegrees=45, twistLimitDegrees=5}
 {bone1Name=dyn_coat_fr_r_02, bone2Name=dyn_coat_fr_r_01, swing1LimitDegrees=45, swing2LimitDegrees=45, twistLimitDegrees=5}
 {bone1Name=dyn_coat_fr_r_03, bone2Name=dyn_coat_fr_r_02, swing1LimitDegrees=45, swing2LimitDegrees=45, twistLimitDegrees=5}
 {bone1Name=dyn_coat_fr_r_04, bone2Name=dyn_coat_fr_r_03, swing1LimitDegrees=45, swing2LimitDegrees=45, twistLimitDegrees=5}
 {bone1Name=dyn_coat_fr_r_05, bone2Name=dyn_coat_fr_r_04, swing1LimitDegrees=45, swing2LimitDegrees=45, twistLimitDegrees=5}
 {
  bone1Name=dyn_coat_r_MSTR
  bone2Name=spine_03
  swing1LimitDegrees=45
  swing1Motion=Free
  swing2LimitDegrees=45
  swing2Motion=Free
  twistLimitDegrees=45
  twistMotion=Free
 }
 {bone1Name=dyn_coat_r_01, bone2Name=dyn_coat_r_MSTR, swing1LimitDegrees=45, swing2LimitDegrees=45, twistLimitDegrees=5}
 {bone1Name=dyn_coat_r_02, bone2Name=dyn_coat_r_01, swing1LimitDegrees=45, swing2LimitDegrees=45, twistLimitDegrees=5}
 {bone1Name=dyn_coat_r_03, bone2Name=dyn_coat_r_02, swing1LimitDegrees=45, swing2LimitDegrees=45, twistLimitDegrees=5}
 {bone1Name=dyn_coat_r_04, bone2Name=dyn_coat_r_03, swing1LimitDegrees=45, swing2LimitDegrees=45, twistLimitDegrees=5}
 {bone1Name=dyn_coat_r_05, bone2Name=dyn_coat_r_04, swing1LimitDegrees=45, swing2LimitDegrees=45, twistLimitDegrees=5}
 {bone1Name=dyn_coat_r_06, bone2Name=dyn_coat_r_05, swing1LimitDegrees=45, swing2LimitDegrees=45, twistLimitDegrees=5}
 {
  bone1Name=dyn_coat_bk_MSTR
  bone2Name=spine_03
  swing1LimitDegrees=45
  swing1Motion=Free
  swing2LimitDegrees=45
  swing2Motion=Free
  twistLimitDegrees=45
  twistMotion=Free
 }
 {bone1Name=dyn_coat_bk_01, bone2Name=dyn_coat_bk_MSTR, swing1LimitDegrees=45, swing2LimitDegrees=45, twistLimitDegrees=5}
 {bone1Name=dyn_coat_bk_02, bone2Name=dyn_coat_bk_01, swing1LimitDegrees=45, swing2LimitDegrees=45, twistLimitDegrees=5}
 {bone1Name=dyn_coat_bk_03, bone2Name=dyn_coat_bk_02, swing1LimitDegrees=45, swing2LimitDegrees=45, twistLimitDegrees=5}
 {bone1Name=dyn_coat_bkend_l_01, bone2Name=dyn_coat_bk_03, swing1LimitDegrees=45, swing2LimitDegrees=45, twistLimitDegrees=5}
 {bone1Name=dyn_coat_bkend_l_02, bone2Name=dyn_coat_bkend_l_01, swing1LimitDegrees=45, swing2LimitDegrees=45, twistLimitDegrees=5}
 {bone1Name=dyn_coat_bkend_l_03, bone2Name=dyn_coat_bkend_l_02, swing1LimitDegrees=45, swing2LimitDegrees=45, twistLimitDegrees=5}
 {bone1Name=dyn_coat_bkend_r_01, bone2Name=dyn_coat_bk_03, swing1LimitDegrees=45, swing2LimitDegrees=45, twistLimitDegrees=5}
 {bone1Name=dyn_coat_bkend_r_02, bone2Name=dyn_coat_bkend_r_01, swing1LimitDegrees=45, swing2LimitDegrees=45, twistLimitDegrees=5}
 {bone1Name=dyn_coat_bkend_r_03, bone2Name=dyn_coat_bkend_r_02, swing1LimitDegrees=45, swing2LimitDegrees=45, twistLimitDegrees=5}
 {
  bone1Name=dyn_coat_bk_r_MSTR
  bone2Name=spine_03
  swing1LimitDegrees=45
  swing1Motion=Free
  swing2LimitDegrees=45
  swing2Motion=Free
  twistLimitDegrees=45
  twistMotion=Free
 }
 {bone1Name=dyn_coat_bk_r_01, bone2Name=dyn_coat_bk_r_MSTR, swing1LimitDegrees=45, swing2LimitDegrees=45, twistLimitDegrees=5}
 {bone1Name=dyn_coat_bk_r_02, bone2Name=dyn_coat_bk_r_01, swing1LimitDegrees=45, swing2LimitDegrees=45, twistLimitDegrees=5}
 {bone1Name=dyn_coat_bk_r_03, bone2Name=dyn_coat_bk_r_02, swing1LimitDegrees=45, swing2LimitDegrees=45, twistLimitDegrees=5}
 {bone1Name=dyn_coat_bk_r_04, bone2Name=dyn_coat_bk_r_03, swing1LimitDegrees=45, swing2LimitDegrees=45, twistLimitDegrees=5}
 {bone1Name=dyn_coat_bk_r_05, bone2Name=dyn_coat_bk_r_04, swing1LimitDegrees=45, swing2LimitDegrees=45, twistLimitDegrees=5}
 {bone1Name=dyn_coat_bk_r_06, bone2Name=dyn_coat_bk_r_05, swing1LimitDegrees=45, swing2LimitDegrees=45, twistLimitDegrees=5}
 {
  bone1Name=spine_01
  bone2Name=pelvis
  swing1LimitDegrees=45
  swing1Motion=Free
  swing2LimitDegrees=45
  swing2Motion=Free
  twistLimitDegrees=45
  twistMotion=Free
 }
 {
  bone1Name=spine_02
  bone2Name=spine_01
  swing1LimitDegrees=45
  swing1Motion=Free
  swing2LimitDegrees=45
  swing2Motion=Free
  twistLimitDegrees=45
  twistMotion=Free
 }
 {
  bone1Name=spine_03
  bone2Name=spine_02
  swing1LimitDegrees=45
  swing1Motion=Free
  swing2LimitDegrees=45
  swing2Motion=Free
  twistLimitDegrees=45
  twistMotion=Free
 }
 {
  bone1Name=thigh_l
  bone2Name=pelvis
  swing1LimitDegrees=45
  swing1Motion=Free
  swing2LimitDegrees=45
  swing2Motion=Free
  twistLimitDegrees=45
  twistMotion=Free
 }
 {
  bone1Name=calf_l
  bone2Name=thigh_l
  swing1LimitDegrees=45
  swing1Motion=Free
  swing2LimitDegrees=45
  swing2Motion=Free
  twistLimitDegrees=45
  twistMotion=Free
 }
 {
  bone1Name=thigh_r
  bone2Name=pelvis
  swing1LimitDegrees=45
  swing1Motion=Free
  swing2LimitDegrees=45
  swing2Motion=Free
  twistLimitDegrees=45
  twistMotion=Free
 }
 {
  bone1Name=calf_r
  bone2Name=thigh_r
  swing1LimitDegrees=45
  swing1Motion=Free
  swing2LimitDegrees=45
  swing2Motion=Free
  twistLimitDegrees=45
  twistMotion=Free
 }
 {
  bone1Name=dyn_coat_l_03
  bone2Name=dyn_coat_fr_l_03
  swing1LimitDegrees=45
  swing1Motion=Free
  swing2LimitDegrees=45
  swing2Motion=Free
  twistLimitDegrees=45
  twistMotion=Free
 }
 {
  bone1Name=dyn_coat_l_05
  bone2Name=dyn_coat_fr_l_05
  swing1LimitDegrees=45
  swing1Motion=Free
  swing2LimitDegrees=45
  swing2Motion=Free
  twistLimitDegrees=45
  twistMotion=Free
 }
 {
  bone1Name=dyn_coat_bk_l_03
  bone2Name=dyn_coat_l_03
  swing1LimitDegrees=45
  swing1Motion=Free
  swing2LimitDegrees=45
  swing2Motion=Free
  twistLimitDegrees=45
  twistMotion=Free
 }
 {
  bone1Name=dyn_coat_bk_l_05
  bone2Name=dyn_coat_l_06
  swing1LimitDegrees=45
  swing1Motion=Free
  swing2LimitDegrees=45
  swing2Motion=Free
  twistLimitDegrees=45
  twistMotion=Free
 }
 {
  bone1Name=dyn_coat_bkend_l_01
  bone2Name=dyn_coat_bk_l_04
  swing1LimitDegrees=45
  swing1Motion=Free
  swing2LimitDegrees=45
  swing2Motion=Free
  twistLimitDegrees=45
  twistMotion=Free
 }
 {
  bone1Name=dyn_coat_bkend_l_03
  bone2Name=dyn_coat_bk_l_06
  swing1LimitDegrees=45
  swing1Motion=Free
  swing2LimitDegrees=45
  swing2Motion=Free
  twistLimitDegrees=45
  twistMotion=Free
 }
 {
  bone1Name=dyn_coat_r_03
  bone2Name=dyn_coat_fr_r_03
  swing1LimitDegrees=45
  swing1Motion=Free
  swing2LimitDegrees=45
  swing2Motion=Free
  twistLimitDegrees=45
  twistMotion=Free
 }
 {
  bone1Name=dyn_coat_r_05
  bone2Name=dyn_coat_fr_r_05
  swing1LimitDegrees=45
  swing1Motion=Free
  swing2LimitDegrees=45
  swing2Motion=Free
  twistLimitDegrees=45
  twistMotion=Free
 }
 {
  bone1Name=dyn_coat_bk_r_03
  bone2Name=dyn_coat_r_03
  swing1LimitDegrees=45
  swing1Motion=Free
  swing2LimitDegrees=45
  swing2Motion=Free
  twistLimitDegrees=45
  twistMotion=Free
 }
 {
  bone1Name=dyn_coat_bk_r_05
  bone2Name=dyn_coat_r_06
  swing1LimitDegrees=45
  swing1Motion=Free
  swing2LimitDegrees=45
  swing2Motion=Free
  twistLimitDegrees=45
  twistMotion=Free
 }
 {
  bone1Name=dyn_coat_bkend_r_01
  bone2Name=dyn_coat_bk_r_04
  swing1LimitDegrees=45
  swing1Motion=Free
  swing2LimitDegrees=45
  swing2Motion=Free
  twistLimitDegrees=45
  twistMotion=Free
 }
 {
  bone1Name=dyn_coat_bkend_r_03
  bone2Name=dyn_coat_bk_r_06
  swing1LimitDegrees=45
  swing1Motion=Free
  swing2LimitDegrees=45
  swing2Motion=Free
  twistLimitDegrees=45
  twistMotion=Free
 }
 {bone1Name=dyn_coat_fr_r_06, bone2Name=dyn_coat_fr_r_05, swing1LimitDegrees=45, swing2LimitDegrees=45, twistLimitDegrees=5}
]
