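# Physics asset for the dangling chest zipper on the male medium Racer Zero Mask character body.
# A large spine_05 capsule covers the upper torso, with two small capsules for the zipper master and zipper bones.
# The master bone's constraint to spine_05 leaves swing and twist free, while the zipper's constraint to the master sets moderate swing limits and a narrow twist limit.
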PhysicsAsset /BRCosmetics/Characters/Player/Male/Medium/Bodies/M_MED_Racer_Zero_Mask/Meshes/M_MED_Racer_Zero_Mask_zip_Dyn
bodies {
 dyn_zipper_chest [{shapeType=Capsule, length=2, radius=0.5, shapeName=dyn_zipper_chest_capsule, center=<2 0 0>, rotation=<rot 90 0 0>}]
 dyn_zipper_chest_MSTR [{shapeType=Capsule, length=0.505, radius=0.505}]
 spine_05 [{shapeType=Capsule, length=8.7745, radius=10.8041, center=<-0.428 0.7046 0>}]
}
constraints [
 {
  bone1Name=dyn_zipper_chest_MSTR
  bone2Name=spine_05
  swing1LimitDegrees=45
  swing1Motion=Free
  swing2LimitDegrees=45
  swing2Motion=Free
  twistLimitDegrees=45
  twistMotion=Free
 }
 {bone1Name=dyn_zipper_chest, bone2Name=dyn_zipper_chest_MSTR, swing1LimitDegrees=30, swing2LimitDegrees=30, twistLimitDegrees=5}
]
