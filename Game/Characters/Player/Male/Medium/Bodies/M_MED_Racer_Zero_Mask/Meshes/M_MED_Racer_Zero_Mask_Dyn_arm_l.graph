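# Physics asset for the left arm pad dynamic bones on the medium male Racer Zero Mask player body.
# Three box bodies form a chain along the pad bones, anchored to a capsule on the left upper-arm twist bone.
# Narrow angular limits hold the pad close to the arm, and the two outer joints also lock their first swing axis.
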PhysicsAsset /Game/Characters/Player/Male/Medium/Bodies/M_MED_Racer_Zero_Mask/Meshes/M_MED_Racer_Zero_Mask_Dyn_arm_l
bodies {
 dyn_arm_pad_l_01 [{shapeType=Box, extentX=7.5, extentY=13.0996, extentZ=1.8698, center=<5.4514 0.0004 0>}]
 dyn_arm_pad_l_02 [{shapeType=Box, extentX=7.5, extentY=13, extentZ=6, center=<5.4514 0 2.45>}]
 dyn_arm_pad_l_03 [{shapeType=Box, extentX=7.5, extentY=13, extentZ=6, center=<5.4514 0 2.45>}]
 upperarm_twist_01_l [{shapeType=Capsule, length=12.2606, radius=9, center=<13.1719 0.001 -4>, rotation=<rot 90 0 0>}]
}
constraints [
 {bone1Name=dyn_arm_pad_l_01, bone2Name=upperarm_twist_01_l, swing1LimitDegrees=1, swing2LimitDegrees=8, twistLimitDegrees=1}
 {
  bone1Name=dyn_arm_pad_l_02
  bone2Name=dyn_arm_pad_l_01
  swing1LimitDegrees=10
  swing1Motion=Locked
  swing2LimitDegrees=3
  twistLimitDegrees=1
 }
 {
  bone1Name=dyn_arm_pad_l_03
  bone2Name=dyn_arm_pad_l_02
  swing1LimitDegrees=10
  swing1Motion=Locked
  swing2LimitDegrees=3
  twistLimitDegrees=1
 }
]
