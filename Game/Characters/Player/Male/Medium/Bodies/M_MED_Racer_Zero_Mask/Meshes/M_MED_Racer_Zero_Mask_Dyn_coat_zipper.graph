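# Physics asset for the dynamic coat zipper on the male medium Racer Zero Mask player body.
# A box body on dyn_zipper_coat, offset along the bone's X axis, covers the zipper, and a small capsule sits on dyn_zipper_coat_MSTR.
# One constraint joins the two with both swing axes and twist set to free, so the listed angular limits do not restrict rotation.
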PhysicsAsset /Game/Characters/Player/Male/Medium/Bodies/M_MED_Racer_Zero_Mask/Meshes/M_MED_Racer_Zero_Mask_Dyn_coat_zipper
bodies {
 dyn_zipper_coat [{shapeType=Box, extentX=2.7465, extentY=1.2344, extentZ=0.5, center=<1.5958 0.0001 0>}]
 dyn_zipper_coat_MSTR [{shapeType=Capsule, length=0.505, radius=0.505}]
}
constraints [{
  bone1Name=dyn_zipper_coat
  bone2Name=dyn_zipper_coat_MSTR
  swing1LimitDegrees=45
  swing1Motion=Free
  swing2LimitDegrees=45
  swing2Motion=Free
  twistLimitDegrees=45
  twistMotion=Free
 }]
