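# Physics asset for the dynamic trail pieces on the chest and waist of the male Racer Zero Mask player mesh.
# Six three-link capsule chains hang from small MSTR bodies joined to spine_03, spine_04 and spine_05, with wide swing and tight twist limits per link.
# Larger capsules on the spine and clavicles approximate the torso, and their joints, like the MSTR attachments, leave rotation free.
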PhysicsAsset /Game/Characters/Player/Male/Medium/Bodies/M_MED_Racer_Zero_Mask/Meshes/M_MED_Racer_Zero_Mask_Dyn_fx
bodies {
 clavicle_l [{shapeType=Capsule, length=10.0938, radius=10.0312, center=<12.1861 0.4203 -1.1838>, rotation=<rot 90 0 0>}]
 clavicle_r [{shapeType=Capsule, length=10.0938, radius=10.0312, center=<-12.2284 -0.4224 1.181>, rotation=<rot 90 0 0>}]
 dyn_fx_chest_trail_bk_l_01 [{shapeType=Capsule, length=5.0363, radius=1.7237, center=<5 0 0>, rotation=<rot 90 0 0>}]
 dyn_fx_chest_trail_bk_l_02 [{shapeType=Capsule, length=5.0363, radius=1.7237, center=<5 0 0>, rotation=<rot 90 0 0>}]
 dyn_fx_chest_trail_bk_l_03 [{shapeType=Capsule, length=5.0363, radius=1.7237, center=<5 0 0>, rotation=<rot 90 0 0>}]
 dyn_fx_chest_trail_bk_l_MSTR [{shapeType=Capsule, length=0.5, radius=0.5}]
 dyn_fx_chest_trail_bk_r_01 [{shapeType=Capsule, length=5.0363, radius=1.7237, center=<4.4228 0 0>, rotation=<rot 90 0 0>}]
 dyn_fx_chest_trail_bk_r_02 [{shapeType=Capsule, length=5.0363, radius=1.7237, center=<4.4228 0 0>, rotation=<rot 90 0 0>}]
 dyn_fx_chest_trail_bk_r_03 [{shapeType=Capsule, length=5.0363, radius=1.7237, center=<4.4228 0 0>, rotation=<rot 90 0 0>}]
 dyn_fx_chest_trail_bk_r_MSTR [{shapeType=Capsule, length=0.5, radius=0.5}]
 dyn_fx_chest_trail_fr_l_01 [{shapeType=Capsule, length=4.1613, radius=1.7237, center=<1.8506 0.0001 0>, rotation=<rot 90 0 0>}]
 dyn_fx_chest_trail_fr_l_02 [{shapeType=Capsule, length=5.255, radius=1.505, center=<3 0 0>, rotation=<rot 90 0 0>}]
 dyn_fx_chest_trail_fr_l_03 [{shapeType=Capsule, length=5.0363, radius=1.7237, center=<3 0 0>, rotation=<rot 90 0 0>}]
 dyn_fx_chest_trail_fr_l_MSTR [{shapeType=Capsule, length=0.5, radius=0.5}]
 dyn_fx_chest_trail_fr_r_01 [{shapeType=Capsule, length=3.4113, radius=1.7237, center=<4.5462 -0.0001 0>, rotation=<rot 90 0 0>}]
 dyn_fx_chest_trail_fr_r_02 [{shapeType=Capsule, length=5.0363, radius=1.7237, center=<6 0 0>, rotation=<rot 90 0 0>}]
 dyn_fx_chest_trail_fr_r_03 [{shapeType=Capsule, length=5.0363, radius=1.7237, center=<6 0 0>, rotation=<rot 90 0 0>}]
 dyn_fx_chest_trail_fr_r_MSTR [{shapeType=Capsule, length=0.5, radius=0.5}]
 dyn_fx_waist_trail_fr_l_01 [{shapeType=Capsule, length=5.0363, radius=1.7237, center=<4.4228 0 0>, rotation=<rot 90 0 0>}]
 dyn_fx_waist_trail_fr_l_02 [{shapeType=Capsule, length=5.0363, radius=1.7237, center=<4.4228 0 0>, rotation=<rot 90 0 0>}]
 dyn_fx_waist_trail_fr_l_03 [{shapeType=Capsule, length=5.7238, radius=1.0362, center=<4.4228 0 0>, rotation=<rot 90 0 0>}]
 dyn_fx_waist_trail_fr_l_MSTR [{shapeType=Capsule, length=0.5, radius=0.5}]
 dyn_fx_waist_trail_fr_r_01 [{shapeType=Capsule, length=5.0363, radius=1.7237, center=<4.4228 0 0>, rotation=<rot 90 0 0>}]
 dyn_fx_waist_trail_fr_r_02 [{shapeType=Capsule, length=5.0363, radius=1.7237, center=<4.4228 0 0>, rotation=<rot 90 0 0>}]
 dyn_fx_waist_trail_fr_r_03 [{shapeType=Capsule, length=5.7238, radius=1.0362, center=<4.4228 0 0>, rotation=<rot 90 0 0>}]
 dyn_fx_waist_trail_fr_r_MSTR [{shapeType=Capsule, length=0.5, radius=0.5}]
 spine_02 [{shapeType=Capsule, length=6.8438, radius=11.7812, center=<2.6694 0 0>}]
 spine_03 [{shapeType=Capsule, length=7.5625, radius=11.7812, center=<2.6694 0 0>}]
 spine_04 [{shapeType=Capsule, length=10.4688, radius=11.7812, center=<2.6694 0 0>}]
 spine_05 [
  {shapeType=Capsule, length=14.7188, radius=11.7812, center=<2.6694 0.2919 0>}
  {shapeType=Capsule, length=11.25, radius=7.2812, center=<0 10.7536 0.4193>, rotation=<rot 90 0 0>}
  {shapeType=Capsule, length=7.3125, radius=10.0312, center=<12.5159 13.1028 -0.709>, rotation=<rot 90 0 0>}
 ]
}
constraints [
 {
  bone1Name=dyn_fx_chest_trail_bk_l_01
  bone2Name=dyn_fx_chest_trail_bk_l_MSTR
  swing1LimitDegrees=60
  swing2LimitDegrees=60
  twistLimitDegrees=5
 }
 {
  bone1Name=dyn_fx_chest_trail_bk_l_02
  bone2Name=dyn_fx_chest_trail_bk_l_01
  swing1LimitDegrees=60
  swing2LimitDegrees=60
  twistLimitDegrees=5
 }
 {
  bone1Name=dyn_fx_chest_trail_bk_l_03
  bone2Name=dyn_fx_chest_trail_bk_l_02
  swing1LimitDegrees=60
  swing2LimitDegrees=60
  twistLimitDegrees=5
 }
 {
  bone1Name=dyn_fx_chest_trail_bk_r_01
  bone2Name=dyn_fx_chest_trail_bk_r_MSTR
  swing1LimitDegrees=60
  swing2LimitDegrees=60
  twistLimitDegrees=5
 }
 {
  bone1Name=dyn_fx_chest_trail_bk_r_02
  bone2Name=dyn_fx_chest_trail_bk_r_01
  swing1LimitDegrees=60
  swing2LimitDegrees=60
  twistLimitDegrees=5
 }
 {
  bone1Name=dyn_fx_chest_trail_bk_r_03
  bone2Name=dyn_fx_chest_trail_bk_r_02
  swing1LimitDegrees=60
  swing2LimitDegrees=60
  twistLimitDegrees=5
 }
 {
  bone1Name=dyn_fx_chest_trail_fr_l_01
  bone2Name=dyn_fx_chest_trail_fr_l_MSTR
  swing1LimitDegrees=60
  swing2LimitDegrees=60
  twistLimitDegrees=5
 }
 {
  bone1Name=dyn_fx_chest_trail_fr_l_02
  bone2Name=dyn_fx_chest_trail_fr_l_01
  swing1LimitDegrees=60
  swing2LimitDegrees=60
  twistLimitDegrees=5
 }
 {
  bone1Name=dyn_fx_chest_trail_fr_l_03
  bone2Name=dyn_fx_chest_trail_fr_l_02
  swing1LimitDegrees=60
  swing2LimitDegrees=60
  twistLimitDegrees=5
 }
 {
  bone1Name=dyn_fx_chest_trail_fr_r_01
  bone2Name=dyn_fx_chest_trail_fr_r_MSTR
  swing1LimitDegrees=60
  swing2LimitDegrees=60
  twistLimitDegrees=5
 }
 {
  bone1Name=dyn_fx_chest_trail_fr_r_02
  bone2Name=dyn_fx_chest_trail_fr_r_01
  swing1LimitDegrees=60
  swing2LimitDegrees=60
  twistLimitDegrees=5
 }
 {
  bone1Name=dyn_fx_chest_trail_fr_r_03
  bone2Name=dyn_fx_chest_trail_fr_r_02
  swing1LimitDegrees=60
  swing2LimitDegrees=60
  twistLimitDegrees=5
 }
 {
  bone1Name=dyn_fx_waist_trail_fr_l_01
  bone2Name=dyn_fx_waist_trail_fr_l_MSTR
  swing1LimitDegrees=60
  swing2LimitDegrees=60
  twistLimitDegrees=5
 }
 {
  bone1Name=dyn_fx_waist_trail_fr_l_02
  bone2Name=dyn_fx_waist_trail_fr_l_01
  swing1LimitDegrees=60
  swing2LimitDegrees=60
  twistLimitDegrees=5
 }
 {
  bone1Name=dyn_fx_waist_trail_fr_l_03
  bone2Name=dyn_fx_waist_trail_fr_l_02
  swing1LimitDegrees=60
  swing2LimitDegrees=60
  twistLimitDegrees=5
 }
 {
  bone1Name=dyn_fx_waist_trail_fr_r_01
  bone2Name=dyn_fx_waist_trail_fr_r_MSTR
  swing1LimitDegrees=60
  swing2LimitDegrees=60
  twistLimitDegrees=5
 }
 {
  bone1Name=dyn_fx_waist_trail_fr_r_02
  bone2Name=dyn_fx_waist_trail_fr_r_01
  swing1LimitDegrees=60
  swing2LimitDegrees=60
  twistLimitDegrees=5
 }
 {
  bone1Name=dyn_fx_waist_trail_fr_r_03
  bone2Name=dyn_fx_waist_trail_fr_r_02
  swing1LimitDegrees=60
  swing2LimitDegrees=60
  twistLimitDegrees=5
 }
 {
  bone1Name=dyn_fx_chest_trail_bk_l_MSTR
  bone2Name=spine_05
  swing1LimitDegrees=45
  swing1Motion=Free
  swing2LimitDegrees=45
  swing2Motion=Free
  twistLimitDegrees=45
  twistMotion=Free
 }
 {
  bone1Name=dyn_fx_chest_trail_bk_r_MSTR
  bone2Name=spine_05
  swing1LimitDegrees=45
  swing1Motion=Free
  swing2LimitDegrees=45
  swing2Motion=Free
  twistLimitDegrees=45
  twistMotion=Free
 }
 {
  bone1Name=dyn_fx_chest_trail_fr_l_MSTR
  bone2Name=spine_04
  swing1LimitDegrees=45
  swing1Motion=Free
  swing2LimitDegrees=45
  swing2Motion=Free
  twistLimitDegrees=45
  twistMotion=Free
 }
 {
  bone1Name=dyn_fx_chest_trail_fr_r_MSTR
  bone2Name=spine_04
  swing1LimitDegrees=45
  swing1Motion=Free
  swing2LimitDegrees=45
  swing2Motion=Free
  twistLimitDegrees=45
  twistMotion=Free
 }
 {
  bone1Name=dyn_fx_waist_trail_fr_l_MSTR
  bone2Name=spine_03
  swing1LimitDegrees=45
  swing1Motion=Free
  swing2LimitDegrees=45
  swing2Motion=Free
  twistLimitDegrees=45
  twistMotion=Free
 }
 {
  bone1Name=dyn_fx_waist_trail_fr_r_MSTR
  bone2Name=spine_03
  swing1LimitDegrees=45
  swing1Motion=Free
  swing2LimitDegrees=45
  swing2Motion=Free
  twistLimitDegrees=45
  twistMotion=Free
 }
 {
  bone1Name=spine_03
  bone2Name=spine_02
  swing1LimitDegrees=45
  swing1Motion=Free
  swing2LimitDegrees=45
  swing2Motion=Free
  twistLimitDegrees=45
  twistMotion=Free
 }
 {
  bone1Name=spine_04
  bone2Name=spine_03
  swing1LimitDegrees=45
  swing1Motion=Free
  swing2LimitDegrees=45
  swing2Motion=Free
  twistLimitDegrees=45
  twistMotion=Free
 }
 {
  bone1Name=spine_05
  bone2Name=spine_04
  swing1LimitDegrees=45
  swing1Motion=Free
  swing2LimitDegrees=45
  swing2Motion=Free
  twistLimitDegrees=45
  twistMotion=Free
 }
 {
  bone1Name=clavicle_l
  bone2Name=spine_05
  swing1LimitDegrees=45
  swing1Motion=Free
  swing2LimitDegrees=45
  swing2Motion=Free
  twistLimitDegrees=45
  twistMotion=Free
 }
 {
  bone1Name=clavicle_r
  bone2Name=spine_05
  swing1LimitDegrees=45
  swing1Motion=Free
  swing2LimitDegrees=45
  swing2Motion=Free
  twistLimitDegrees=45
  twistMotion=Free
 }
]
