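# Physics asset for the left-leg dynamic shoe parts of the medium male Racer Zero Mask player character.
# Capsule bodies cover the calf, the shoe heel and tongue bones, and a small master body for each, with two capsules shaping the tongue.
# Heel and tongue each constrain to their master with tight twist and moderate swing limits, while both masters connect to the calf with free swing and twist.
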PhysicsAsset /Game/Characters/Player/Male/Medium/Bodies/M_MED_Racer_Zero_Mask/Meshes/M_MED_Racer_Zero_Mask_Dyn_leg_l
bodies {
 calf_l [{shapeType=Capsule, length=2.3856, radius=2.5778, center=<-33.6406 1.0186 -0.0614>, rotation=<rot 90 -90 -92.1447>}]
 dyn_shoe_heel_l_01 [{shapeType=Capsule, length=3.125, radius=0.875, center=<1 0 0.0001>, rotation=<rot 90 0 0>}]
 dyn_shoe_heel_l_MSTR [{shapeType=Capsule, length=0.505, radius=0.505}]
 dyn_shoe_tongue_l [
  {shapeType=Capsule, length=3.9284, radius=1.0055, center=<4.4 0 0.0001>, rotation=<rot 90 0 0>}
  {shapeType=Capsule, length=1.7181, radius=1.0055, center=<1.8319 -4.5886 -2.4023>, rotation=<rot 67.1214 -82.2483 -60.3503>}
 ]
 dyn_shoe_tongue_l_MSTR [{shapeType=Capsule, length=0.505, radius=0.505}]
}
constraints [
 {bone1Name=dyn_shoe_tongue_l, bone2Name=dyn_shoe_tongue_l_MSTR, swing1LimitDegrees=20, swing2LimitDegrees=45, twistLimitDegrees=5}
 {bone1Name=dyn_shoe_heel_l_01, bone2Name=dyn_shoe_heel_l_MSTR, swing1LimitDegrees=20, swing2LimitDegrees=30, twistLimitDegrees=5}
 {
  bone1Name=dyn_shoe_tongue_l_MSTR
  bone2Name=calf_l
  swing1LimitDegrees=45
  swing1Motion=Free
  swing2LimitDegrees=45
  swing2Motion=Free
  twistLimitDegrees=45
  twistMotion=Free
 }
 {
  bone1Name=dyn_shoe_heel_l_MSTR
  bone2Name=calf_l
  swing1LimitDegrees=45
  swing1Motion=Free
  swing2LimitDegrees=45
  swing2Motion=Free
  twistLimitDegrees=45
  twistMotion=Free
 }
]
