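# Post-body animation Blueprint for the Racer Zero Mask player cosmetic, adding procedural secondary motion to its coat and other dynamic parts.
# One cached-pose branch constrains coat bones to thighs, aims master bones off the pelvis, then runs per-bone spring-damped AnimDynamics.
# Another runs a skydiving RigidBody simulation with variable-driven gravity, then pitches rear coat bones.
# Further RigidBody nodes use separate physics assets for zipper, limbs and FX.
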AnimBlueprint /BRCosmetics/Characters/Player/Male/Medium/Bodies/M_MED_Racer_Zero_Mask/Meshes/M_MED_Racer_Zero_Mask_AnimBP
target_skeleton=&/M_MED_Racer_Zero_Mask_Skeleton
graphs [
 AnimGraph {nodes [
   n1 = "Transform (Modify) Bone - Bone: dyn_coat_bk_r_03"!(ComponentPose, Alpha=1, Translation="0,0,1", Rotation="20,2,-35") -> (Pose) {
    alpha=0
    rotationMode=BMM_Additive
    rotationSpace=BCS_BoneSpace
    translationMode=BMM_Replace
    translationSpace=BCS_BoneSpace
    alphaBoolBlend {blendOption=Linear}
    alphaScaleBiasClamp {bInterpResult=true, interpSpeedIncreasing=20}
    boneToModify {boneName=dyn_coat_bk_r_03}
    rotation=<rot 20 2 -35>
    translation {z=1}
   }
   n2 = "Transform (Modify) Bone - Bone: dyn_coat_r_03"!(ComponentPose=@n1, Alpha=1, Translation="5,0,8", Rotation="-20,-6,5.5") -> (Pose) {
    alpha=0
    rotationMode=BMM_Additive
    rotationSpace=BCS_BoneSpace
    translationMode=BMM_Replace
    translationSpace=BCS_BoneSpace
    alphaBoolBlend {blendOption=Linear}
    alphaScaleBiasClamp {bInterpResult=true, interpSpeedIncreasing=20}
    boneToModify {boneName=dyn_coat_r_03}
    rotation=<rot -20 -6 5.5>
    translation {x=5, z=8}
   }
   "Transform (Modify) Bone - Bone: dyn_coat_bk_r_05"!(ComponentPose=@n2, Alpha=1, Translation="-7,1.5,0", Rotation="0.13,-35,0.2") -> (Pose) {
    alpha=0
    rotationMode=BMM_Additive
    rotationSpace=BCS_BoneSpace
    translationMode=BMM_Replace
    translationSpace=BCS_BoneSpace
    alphaBoolBlend {blendOption=Linear}
    alphaScaleBiasClamp {bInterpResult=true, interpSpeedIncreasing=20}
    boneToModify {boneName=dyn_coat_bk_r_05}
    rotation=<rot 0.13 -35 0.2>
    translation=<-7 1.5>
   }
   n3 = RigidBody() {
    bClampLinearTranslationLimitToRefPose=true
    bOverrideWorldGravity=true
    overridePhysicsAsset=/Game/Characters/Player/Male/Medium/Bodies/M_MED_Racer_Zero_Mask/Meshes/M_MED_Racer_Zero_Mask_Dyn_coat_zipper
    simulationSpace=BaseBoneSpace
    alphaBoolBlend {blendOption=Linear}
    baseBoneRef {boneName=dyn_coat_fr_l_04}
    componentAppliedLinearAccClamp=<3000 3000 3000>
    componentLinearAccScale=<0.4 0.4 0.4>
    componentLinearVelScale=<0.4 0.4 0.4>
    overrideWorldGravity {z=-980}
   }
   n4 = RigidBody(ComponentPose=@n3) {
    bClampLinearTranslationLimitToRefPose=true
    bOverrideWorldGravity=true
    overridePhysicsAsset=/Game/Characters/Player/Male/Medium/Bodies/M_MED_Racer_Zero_Mask/Meshes/M_MED_Racer_Zero_Mask_Dyn_leg_l
    simulationSpace=BaseBoneSpace
    alphaBoolBlend {blendOption=Linear}
    baseBoneRef {boneName=calf_l}
    componentAppliedLinearAccClamp=<3000 3000 3000>
    componentLinearAccScale=<0.4 0.3 0.3>
    componentLinearVelScale=<0.4 0.3 0.3>
    overrideWorldGravity {z=-980}
   }
   n5 = RigidBody(ComponentPose=@n4) {
    bClampLinearTranslationLimitToRefPose=true
    bOverrideWorldGravity=true
    overridePhysicsAsset=/Game/Characters/Player/Male/Medium/Bodies/M_MED_Racer_Zero_Mask/Meshes/M_MED_Racer_Zero_Mask_Dyn_leg_r
    simulationSpace=BaseBoneSpace
    alphaBoolBlend {blendOption=Linear}
    baseBoneRef {boneName=calf_r}
    componentAppliedLinearAccClamp=<3000 3000 3000>
    componentLinearAccScale=<0.4 0.3 0.3>
    componentLinearVelScale=<0.4 0.3 0.3>
    overrideWorldGravity {z=-980}
   }
   n6 = RigidBody(ComponentPose=@n5) {
    bClampLinearTranslationLimitToRefPose=true
    bOverrideWorldGravity=true
    overridePhysicsAsset=/Game/Characters/Player/Male/Medium/Bodies/M_MED_Racer_Zero_Mask/Meshes/M_MED_Racer_Zero_Mask_Dyn_arm_l
    simulationSpace=BaseBoneSpace
    alphaBoolBlend {blendOption=Linear}
    baseBoneRef {boneName=upperarm_l}
    componentAppliedLinearAccClamp=<3000 3000 3000>
    componentLinearAccScale=<0.4 0.3 0.3>
    componentLinearVelScale=<0.4 0.3 0.3>
    overrideWorldGravity {z=-980}
   }
   RigidBody(ComponentPose=@n6) {
    bClampLinearTranslationLimitToRefPose=true
    bOverrideWorldGravity=true
    overridePhysicsAsset=/Game/Characters/Player/Male/Medium/Bodies/M_MED_Racer_Zero_Mask/Meshes/M_MED_Racer_Zero_Mask_Dyn_arm_r
    simulationSpace=BaseBoneSpace
    alphaBoolBlend {blendOption=Linear}
    baseBoneRef {boneName=upperarm_r}
    componentAppliedLinearAccClamp=<3000 3000 3000>
    componentLinearAccScale=<0.4 0.3 0.3>
    componentLinearVelScale=<0.4 0.3 0.3>
    overrideWorldGravity {z=-980}
   }
   RigidBody() {
    bClampLinearTranslationLimitToRefPose=true
    bOverrideWorldGravity=true
    overridePhysicsAsset=/Game/Characters/Player/Male/Medium/Bodies/M_MED_Racer_Zero_Mask/Meshes/M_MED_Racer_Zero_Mask_Dyn_fx
    simulationSpace=BaseBoneSpace
    alphaBoolBlend {blendOption=Linear}
    baseBoneRef {boneName=spine_03}
    componentAppliedLinearAccClamp=<3000 3000 3000>
    componentLinearAccScale=<0.1 0.2 0.2>
    componentLinearVelScale=<0.1 0.2 0.2>
    overrideWorldGravity {z=100}
   }
   n7 = RigidBody!(
    ComponentPose=Local To Component(LocalPose=Use cached pose 'initial Pose'())
    Alpha
    OverrideWorldGravity:Vector=Get gravOverride() -> (gravOverride:Vector)
    ExternalForce
   ) -> (Pose) {
    bClampLinearTranslationLimitToRefPose=true
    bOverrideWorldGravity=true
    overridePhysicsAsset=&/M_MED_Racer_Zero_Mask_Dyn_skydiving
    simulationSpace=BaseBoneSpace
    alphaBoolBlend {blendOption=Linear}
    baseBoneRef {boneName=spine_03}
    componentAppliedLinearAccClamp=<3000 3000 3000>
    componentLinearAccScale=<0.35 0.3 0.3>
    componentLinearVelScale=<0.35 0.3 0.3>
    overrideWorldGravity {z=-980}
   }
   n8 = "Transform (Modify) Bone - Bone: dyn_coat_bk_01"(ComponentPose=@n7, Alpha=1, Translation="0,0,0", Rotation="6,0,0", Scale="1,1,1") {
    rotationMode=BMM_Additive
    rotationSpace=BCS_BoneSpace
    alphaBoolBlend {blendOption=Linear}
    boneToModify {boneName=dyn_coat_bk_01}
    rotation {pitch=6}
   }
   n9 = "Transform (Modify) Bone - Bone: dyn_coat_bk_l_01"(ComponentPose=@n8, Alpha=1, Translation="0,0,0", Rotation="6,0,0", Scale="1,1,1") {
    rotationMode=BMM_Additive
    rotationSpace=BCS_BoneSpace
    alphaBoolBlend {blendOption=Linear}
    boneToModify {boneName=dyn_coat_bk_l_01}
    rotation {pitch=6}
   }
   n10 = "Transform (Modify) Bone - Bone: dyn_coat_bk_r_01"(ComponentPose=@n9, Alpha=1, Translation="0,0,0", Rotation="6,0,0", Scale="1,1,1") {
    rotationMode=BMM_Additive
    rotationSpace=BCS_BoneSpace
    alphaBoolBlend {blendOption=Linear}
    boneToModify {boneName=dyn_coat_bk_r_01}
    rotation {pitch=6}
   }
   n11 = Component To Local(ComponentPose=@n10)
   n12 = "Constraint - dyn_coat_fr_l_03 (P)"(ComponentPose=Local To Component(LocalPose=Use cached pose 'initial Pose'()), Alpha=1, ConstraintWeights_0=1) {
    alphaBoolBlend {blendOption=Linear}
    boneToModify {boneName=dyn_coat_fr_l_03}
    constraintSetup [{targetBone {boneName=thigh_l}}]
    constraintWeights [1]
   }
   n13 = "Constraint - dyn_coat_l_03 (P)"(ComponentPose=@n12, Alpha=1, ConstraintWeights_0=1) {
    alphaBoolBlend {blendOption=Linear}
    boneToModify {boneName=dyn_coat_l_03}
    constraintSetup [{targetBone {boneName=thigh_l}}]
    constraintWeights [1]
   }
   n14 = "Constraint - dyn_coat_bk_l_03 (P)"(ComponentPose=@n13, Alpha=1, ConstraintWeights_0=1) {
    alphaBoolBlend {blendOption=Linear}
    boneToModify {boneName=dyn_coat_bk_l_03}
    constraintSetup [{targetBone {boneName=thigh_l}}]
    constraintWeights [1]
   }
   n15 = "Constraint - dyn_coat_fr_r_03 (P)"(ComponentPose=@n14, Alpha=1, ConstraintWeights_0=1) {
    alphaBoolBlend {blendOption=Linear}
    boneToModify {boneName=dyn_coat_fr_r_03}
    constraintSetup [{targetBone {boneName=thigh_r}}]
    constraintWeights [1]
   }
   n16 = "Constraint - dyn_coat_r_03 (P)"(ComponentPose=@n15, Alpha=1, ConstraintWeights_0=1) {
    alphaBoolBlend {blendOption=Linear}
    boneToModify {boneName=dyn_coat_r_03}
    constraintSetup [{targetBone {boneName=thigh_r}}]
    constraintWeights [1]
   }
   n17 = "Constraint - dyn_coat_bk_r_03 (P)"(ComponentPose=@n16, Alpha=1, ConstraintWeights_0=1) {
    alphaBoolBlend {blendOption=Linear}
    boneToModify {boneName=dyn_coat_bk_r_03}
    constraintSetup [{targetBone {boneName=thigh_r}}]
    constraintWeights [1]
   }
   n18 = Look At - Bone: dyn_coat_fr_l_MSTR(ComponentPose=@n17, Alpha=1) {
    alphaBoolBlend {blendOption=Linear}
    boneToModify {boneName=dyn_coat_fr_l_MSTR}
    lookAtLocation=<-100 30 -40>
    lookAtTarget {boneReference {boneName=pelvis}}
    lookAt_Axis {axis=<1 0>}
   }
   n19 = Look At - Bone: dyn_coat_l_MSTR(ComponentPose=@n18, Alpha=1) {
    alphaBoolBlend {blendOption=Linear}
    boneToModify {boneName=dyn_coat_l_MSTR}
    lookAtLocation=<-100 0 -45>
    lookAtTarget {boneReference {boneName=pelvis}}
    lookAt_Axis {axis=<1 0>}
   }
   n20 = Look At - Bone: dyn_coat_bk_l_MSTR(ComponentPose=@n19, Alpha=1) {
    alphaBoolBlend {blendOption=Linear}
    boneToModify {boneName=dyn_coat_bk_l_MSTR}
    lookAtLocation=<-100 -40 -80>
    lookAtTarget {boneReference {boneName=pelvis}}
    lookAt_Axis {axis=<1 0>}
   }
   n21 = Look At - Bone: dyn_coat_fr_r_MSTR(ComponentPose=@n20, Alpha=1) {
    alphaBoolBlend {blendOption=Linear}
    boneToModify {boneName=dyn_coat_fr_r_MSTR}
    lookAtLocation=<-100 30 45>
    lookAtTarget {boneReference {boneName=pelvis}}
    lookAt_Axis {axis=<1 0>}
   }
   n22 = Look At - Bone: dyn_coat_r_MSTR(ComponentPose=@n21, Alpha=1) {
    alphaBoolBlend {blendOption=Linear}
    boneToModify {boneName=dyn_coat_r_MSTR}
    lookAtLocation {x=-100, z=45}
    lookAtTarget {boneReference {boneName=pelvis}}
    lookAt_Axis {axis=<1 0>}
   }
   n23 = Look At - Bone: dyn_coat_bk_r_MSTR(ComponentPose=@n22, Alpha=1) {
    alphaBoolBlend {blendOption=Linear}
    boneToModify {boneName=dyn_coat_bk_r_MSTR}
    lookAtLocation=<-100 -40 80>
    lookAtTarget {boneReference {boneName=pelvis}}
    lookAt_Axis {axis=<1 0>}
   }
   n24 = Anim Dynamics - Bone: dyn_coat_fr_l_02(ComponentPose=@n23, Alpha=1, ExternalForce="0,0,0") {
    angularDampingOverride=0.9999
    angularSpringConstant=500
    bAngularSpring=true
    bGravityOverrideInSimSpace=true
    bOverrideAngularDamping=true
    simulationSpace=BoneRelative
    alphaBoolBlend {blendOption=Linear}
    boundBone {boneName=dyn_coat_fr_l_02}
    componentAppliedLinearAccClamp=<5000 5000 5000>
    componentLinearAccScale=<0.5 0.5 0.7>
    componentLinearVelScale=<0.5 0.5 0.7>
    physicsBodyDefinitions [{
      sphereCollisionRadius=0
      boundBone {boneName=dyn_coat_fr_l_02}
      boxExtents=<15 20 20>
      constraintSetup {angularLimitsMax=<1 3 15>, angularLimitsMin=<-1 -20 -15>, angularTarget=<0 0 0>}
      localJointOffset=<5 0 0>
     }]
    relativeSpaceBone {boneName=root}
   }
   n25 = Anim Dynamics - Bone: dyn_coat_fr_l_03(ComponentPose=@n24, Alpha=1, ExternalForce="0,0,0") {
    angularDampingOverride=0.9999
    angularSpringConstant=500
    bAngularSpring=true
    bGravityOverrideInSimSpace=true
    bOverrideAngularDamping=true
    simulationSpace=BoneRelative
    alphaBoolBlend {blendOption=Linear}
    boundBone {boneName=dyn_coat_fr_l_03}
    componentAppliedLinearAccClamp=<5000 5000 5000>
    componentLinearAccScale=<0.5 0.5 0.7>
    componentLinearVelScale=<0.5 0.5 0.7>
    physicsBodyDefinitions [{
      sphereCollisionRadius=0
      boundBone {boneName=dyn_coat_fr_l_03}
      boxExtents=<15 20 20>
      constraintSetup {angularLimitsMax=<1 3 15>, angularLimitsMin=<-1 -20 -15>, angularTarget=<0 0 0>}
      localJointOffset=<5 0 0>
     }]
    relativeSpaceBone {boneName=root}
   }
   n26 = Anim Dynamics - Bone: dyn_coat_fr_l_04(ComponentPose=@n25, Alpha=1, ExternalForce="0,0,0") {
    angularDampingOverride=0.9999
    angularSpringConstant=500
    bAngularSpring=true
    bGravityOverrideInSimSpace=true
    bOverrideAngularDamping=true
    simulationSpace=BoneRelative
    alphaBoolBlend {blendOption=Linear}
    boundBone {boneName=dyn_coat_fr_l_04}
    componentAppliedLinearAccClamp=<3500 3500 3500>
    componentLinearAccScale=<0.5 0.5 0.7>
    componentLinearVelScale=<0.5 0.5 0.7>
    physicsBodyDefinitions [{
      sphereCollisionRadius=0
      boundBone {boneName=dyn_coat_fr_l_04}
      boxExtents=<15 20 20>
      constraintSetup {angularLimitsMax=<1 3 15>, angularLimitsMin=<-1 -20 -15>, angularTarget=<0 0 0>}
      localJointOffset=<5 0 0>
     }]
    relativeSpaceBone {boneName=root}
   }
   n27 = Anim Dynamics - Bone: dyn_coat_fr_l_05(ComponentPose=@n26, Alpha=1, ExternalForce="0,0,0") {
    angularDampingOverride=0.9999
    angularSpringConstant=500
    bAngularSpring=true
    bGravityOverrideInSimSpace=true
    bOverrideAngularDamping=true
    simulationSpace=BoneRelative
    alphaBoolBlend {blendOption=Linear}
    boundBone {boneName=dyn_coat_fr_l_05}
    componentAppliedLinearAccClamp=<3500 3500 3500>
    componentLinearAccScale=<0.5 0.5 0.7>
    componentLinearVelScale=<0.5 0.5 0.7>
    physicsBodyDefinitions [{
      sphereCollisionRadius=0
      boundBone {boneName=dyn_coat_fr_l_05}
      boxExtents=<15 20 20>
      constraintSetup {angularLimitsMax=<1 3 15>, angularLimitsMin=<-1 -20 -15>, angularTarget=<0 0 0>}
      localJointOffset=<5 0 0>
     }]
    relativeSpaceBone {boneName=root}
   }
   n28 = Anim Dynamics - Bone: dyn_coat_fr_l_06(ComponentPose=@n27, Alpha=1, ExternalForce="0,0,0") {
    angularDampingOverride=0.9999
    angularSpringConstant=500
    bAngularSpring=true
    bGravityOverrideInSimSpace=true
    bOverrideAngularDamping=true
    simulationSpace=BoneRelative
    alphaBoolBlend {blendOption=Linear}
    boundBone {boneName=dyn_coat_fr_l_06}
    componentAppliedLinearAccClamp=<3500 3500 3500>
    componentLinearAccScale=<0.5 0.5 0.7>
    componentLinearVelScale=<0.5 0.5 0.7>
    physicsBodyDefinitions [{
      sphereCollisionRadius=0
      boundBone {boneName=dyn_coat_fr_l_06}
      boxExtents=<15 20 20>
      constraintSetup {angularLimitsMax=<1 3 15>, angularLimitsMin=<-1 -20 -15>, angularTarget=<0 0 0>}
      localJointOffset=<5 0 0>
     }]
    relativeSpaceBone {boneName=root}
   }
   n29 = Anim Dynamics - Bone: dyn_coat_l_02(ComponentPose=@n28, Alpha=1, ExternalForce="0,0,0") {
    angularDampingOverride=0.9999
    angularSpringConstant=500
    bAngularSpring=true
    bGravityOverrideInSimSpace=true
    bOverrideAngularDamping=true
    simulationSpace=BoneRelative
    alphaBoolBlend {blendOption=Linear}
    boundBone {boneName=dyn_coat_l_02}
    componentAppliedLinearAccClamp=<5000 5000 5000>
    componentLinearAccScale=<0.5 0.5 0.7>
    componentLinearVelScale=<0.5 0.5 0.7>
    physicsBodyDefinitions [{
      sphereCollisionRadius=0
      boundBone {boneName=dyn_coat_l_02}
      boxExtents=<15 20 20>
      constraintSetup {angularLimitsMax=<1 3 15>, angularLimitsMin=<-1 -20 -15>, angularTarget=<0 0 0>}
      localJointOffset=<5 0 0>
     }]
    relativeSpaceBone {boneName=root}
   }
   n30 = Anim Dynamics - Bone: dyn_coat_l_03(ComponentPose=@n29, Alpha=1, ExternalForce="0,0,0") {
    angularDampingOverride=0.9999
    angularSpringConstant=500
    bAngularSpring=true
    bGravityOverrideInSimSpace=true
    bOverrideAngularDamping=true
    simulationSpace=BoneRelative
    alphaBoolBlend {blendOption=Linear}
    boundBone {boneName=dyn_coat_l_03}
    componentAppliedLinearAccClamp=<5000 5000 5000>
    componentLinearAccScale=<0.5 0.5 0.7>
    componentLinearVelScale=<0.5 0.5 0.7>
    physicsBodyDefinitions [{
      sphereCollisionRadius=0
      boundBone {boneName=dyn_coat_l_03}
      boxExtents=<15 20 20>
      constraintSetup {angularLimitsMax=<1 3 15>, angularLimitsMin=<-1 -20 -15>, angularTarget=<0 0 0>}
      localJointOffset=<5 0 0>
     }]
    relativeSpaceBone {boneName=root}
   }
   n31 = Anim Dynamics - Bone: dyn_coat_l_04(ComponentPose=@n30, Alpha=1, ExternalForce="0,0,0") {
    angularDampingOverride=0.9999
    angularSpringConstant=500
    bAngularSpring=true
    bGravityOverrideInSimSpace=true
    bOverrideAngularDamping=true
    simulationSpace=BoneRelative
    alphaBoolBlend {blendOption=Linear}
    boundBone {boneName=dyn_coat_l_04}
    componentAppliedLinearAccClamp=<3500 3500 3500>
    componentLinearAccScale=<0.5 0.5 0.7>
    componentLinearVelScale=<0.5 0.5 0.7>
    physicsBodyDefinitions [{
      sphereCollisionRadius=0
      boundBone {boneName=dyn_coat_l_04}
      boxExtents=<15 20 20>
      constraintSetup {angularLimitsMax=<1 3 15>, angularLimitsMin=<-1 -20 -15>, angularTarget=<0 0 0>}
      localJointOffset=<5 0 0>
     }]
    relativeSpaceBone {boneName=root}
   }
   n32 = Anim Dynamics - Bone: dyn_coat_l_05(ComponentPose=@n31, Alpha=1, ExternalForce="0,0,0") {
    angularDampingOverride=0.9999
    angularSpringConstant=500
    bAngularSpring=true
    bGravityOverrideInSimSpace=true
    bOverrideAngularDamping=true
    simulationSpace=BoneRelative
    alphaBoolBlend {blendOption=Linear}
    boundBone {boneName=dyn_coat_l_05}
    componentAppliedLinearAccClamp=<3500 3500 3500>
    componentLinearAccScale=<0.5 0.5 0.7>
    componentLinearVelScale=<0.5 0.5 0.7>
    physicsBodyDefinitions [{
      sphereCollisionRadius=0
      boundBone {boneName=dyn_coat_l_05}
      boxExtents=<15 20 20>
      constraintSetup {angularLimitsMax=<1 3 15>, angularLimitsMin=<-1 -20 -15>, angularTarget=<0 0 0>}
      localJointOffset=<5 0 0>
     }]
    relativeSpaceBone {boneName=root}
   }
   n33 = Anim Dynamics - Bone: dyn_coat_l_06(ComponentPose=@n32, Alpha=1, ExternalForce="0,0,0") {
    angularDampingOverride=0.9999
    angularSpringConstant=500
    bAngularSpring=true
    bGravityOverrideInSimSpace=true
    bOverrideAngularDamping=true
    simulationSpace=BoneRelative
    alphaBoolBlend {blendOption=Linear}
    boundBone {boneName=dyn_coat_l_06}
    componentAppliedLinearAccClamp=<3500 3500 3500>
    componentLinearAccScale=<0.5 0.5 0.7>
    componentLinearVelScale=<0.5 0.5 0.7>
    physicsBodyDefinitions [{
      sphereCollisionRadius=0
      boundBone {boneName=dyn_coat_l_06}
      boxExtents=<15 20 20>
      constraintSetup {angularLimitsMax=<1 3 15>, angularLimitsMin=<-1 -20 -15>, angularTarget=<0 0 0>}
      localJointOffset=<5 0 0>
     }]
    relativeSpaceBone {boneName=root}
   }
   n34 = Anim Dynamics - Bone: dyn_coat_bk_l_02(ComponentPose=@n33, Alpha=1, ExternalForce="0,0,0") {
    angularDampingOverride=0.9999
    angularSpringConstant=500
    bAngularSpring=true
    bGravityOverrideInSimSpace=true
    bOverrideAngularDamping=true
    simulationSpace=BoneRelative
    alphaBoolBlend {blendOption=Linear}
    boundBone {boneName=dyn_coat_bk_l_02}
    componentAppliedLinearAccClamp=<5000 5000 5000>
    componentLinearAccScale=<0.5 0.5 0.7>
    componentLinearVelScale=<0.5 0.5 0.7>
    physicsBodyDefinitions [{
      sphereCollisionRadius=0
      boundBone {boneName=dyn_coat_bk_l_02}
      boxExtents=<15 20 20>
      constraintSetup {angularLimitsMax=<1 3 15>, angularLimitsMin=<-1 -20 -15>, angularTarget=<0 0 0>}
      localJointOffset=<5 0 0>
     }]
    relativeSpaceBone {boneName=root}
   }
   n35 = Anim Dynamics - Bone: dyn_coat_bk_l_03(ComponentPose=@n34, Alpha=1, ExternalForce="0,0,0") {
    angularDampingOverride=0.9999
    angularSpringConstant=500
    bAngularSpring=true
    bGravityOverrideInSimSpace=true
    bOverrideAngularDamping=true
    simulationSpace=BoneRelative
    alphaBoolBlend {blendOption=Linear}
    boundBone {boneName=dyn_coat_bk_l_03}
    componentAppliedLinearAccClamp=<5000 5000 5000>
    componentLinearAccScale=<0.5 0.5 0.7>
    componentLinearVelScale=<0.5 0.5 0.7>
    physicsBodyDefinitions [{
      sphereCollisionRadius=0
      boundBone {boneName=dyn_coat_bk_l_03}
      boxExtents=<15 20 20>
      constraintSetup {angularLimitsMax=<1 3 15>, angularLimitsMin=<-1 -20 -15>, angularTarget=<0 0 0>}
      localJointOffset=<5 0 0>
     }]
    relativeSpaceBone {boneName=root}
   }
   n36 = Anim Dynamics - Bone: dyn_coat_bk_l_04(ComponentPose=@n35, Alpha=1, ExternalForce="0,0,0") {
    angularDampingOverride=0.9999
    angularSpringConstant=500
    bAngularSpring=true
    bGravityOverrideInSimSpace=true
    bOverrideAngularDamping=true
    simulationSpace=BoneRelative
    alphaBoolBlend {blendOption=Linear}
    boundBone {boneName=dyn_coat_bk_l_04}
    componentAppliedLinearAccClamp=<3500 3500 3500>
    componentLinearAccScale=<0.5 0.5 0.7>
    componentLinearVelScale=<0.5 0.5 0.7>
    physicsBodyDefinitions [{
      sphereCollisionRadius=0
      boundBone {boneName=dyn_coat_bk_l_04}
      boxExtents=<15 20 20>
      constraintSetup {angularLimitsMax=<1 3 15>, angularLimitsMin=<-1 -20 -15>, angularTarget=<0 0 0>}
      localJointOffset=<5 0 0>
     }]
    relativeSpaceBone {boneName=root}
   }
   n37 = Anim Dynamics - Bone: dyn_coat_bk_l_05(ComponentPose=@n36, Alpha=1, ExternalForce="0,0,0") {
    angularDampingOverride=0.9999
    angularSpringConstant=500
    bAngularSpring=true
    bGravityOverrideInSimSpace=true
    bOverrideAngularDamping=true
    simulationSpace=BoneRelative
    alphaBoolBlend {blendOption=Linear}
    boundBone {boneName=dyn_coat_bk_l_05}
    componentAppliedLinearAccClamp=<3500 3500 3500>
    componentLinearAccScale=<0.5 0.5 0.7>
    componentLinearVelScale=<0.5 0.5 0.7>
    physicsBodyDefinitions [{
      sphereCollisionRadius=0
      boundBone {boneName=dyn_coat_bk_l_05}
      boxExtents=<15 20 20>
      constraintSetup {angularLimitsMax=<1 3 15>, angularLimitsMin=<-1 -20 -15>, angularTarget=<0 0 0>}
      localJointOffset=<5 0 0>
     }]
    relativeSpaceBone {boneName=root}
   }
   n38 = Anim Dynamics - Bone: dyn_coat_bk_l_06(ComponentPose=@n37, Alpha=1, ExternalForce="0,0,0") {
    angularDampingOverride=0.9999
    angularSpringConstant=500
    bAngularSpring=true
    bGravityOverrideInSimSpace=true
    bOverrideAngularDamping=true
    simulationSpace=BoneRelative
    alphaBoolBlend {blendOption=Linear}
    boundBone {boneName=dyn_coat_bk_l_06}
    componentAppliedLinearAccClamp=<3500 3500 3500>
    componentLinearAccScale=<0.5 0.5 0.7>
    componentLinearVelScale=<0.5 0.5 0.7>
    physicsBodyDefinitions [{
      sphereCollisionRadius=0
      boundBone {boneName=dyn_coat_bk_l_06}
      boxExtents=<15 20 20>
      constraintSetup {angularLimitsMax=<1 3 15>, angularLimitsMin=<-1 -20 -15>, angularTarget=<0 0 0>}
      localJointOffset=<5 0 0>
     }]
    relativeSpaceBone {boneName=root}
   }
   n39 = Anim Dynamics - Bone: dyn_coat_fr_r_02(ComponentPose=@n38, Alpha=1, ExternalForce="0,0,0") {
    angularDampingOverride=0.9999
    angularSpringConstant=500
    bAngularSpring=true
    bGravityOverrideInSimSpace=true
    bOverrideAngularDamping=true
    simulationSpace=BoneRelative
    alphaBoolBlend {blendOption=Linear}
    boundBone {boneName=dyn_coat_fr_r_02}
    componentAppliedLinearAccClamp=<5000 5000 5000>
    componentLinearAccScale=<0.5 0.5 0.7>
    componentLinearVelScale=<0.5 0.5 0.7>
    physicsBodyDefinitions [{
      sphereCollisionRadius=0
      boundBone {boneName=dyn_coat_fr_r_02}
      boxExtents=<15 20 20>
      constraintSetup {angularLimitsMax=<1 3 15>, angularLimitsMin=<-1 -20 -15>, angularTarget=<0 0 0>}
      localJointOffset=<5 0 0>
     }]
    relativeSpaceBone {boneName=root}
   }
   n40 = Anim Dynamics - Bone: dyn_coat_fr_r_03(ComponentPose=@n39, Alpha=1, ExternalForce="0,0,0") {
    angularDampingOverride=0.9999
    angularSpringConstant=500
    bAngularSpring=true
    bGravityOverrideInSimSpace=true
    bOverrideAngularDamping=true
    simulationSpace=BoneRelative
    alphaBoolBlend {blendOption=Linear}
    boundBone {boneName=dyn_coat_fr_r_03}
    componentAppliedLinearAccClamp=<5000 5000 5000>
    componentLinearAccScale=<0.5 0.5 0.7>
    componentLinearVelScale=<0.5 0.5 0.7>
    physicsBodyDefinitions [{
      sphereCollisionRadius=0
      boundBone {boneName=dyn_coat_fr_r_03}
      boxExtents=<15 20 20>
      constraintSetup {angularLimitsMax=<1 8 15>, angularLimitsMin=<-1 -45 -15>, angularTarget=<0 0 0>}
      localJointOffset=<5 0 0>
     }]
    relativeSpaceBone {boneName=root}
   }
   n41 = Anim Dynamics - Bone: dyn_coat_fr_r_04(ComponentPose=@n40, Alpha=1, ExternalForce="0,0,0") {
    angularDampingOverride=0.9999
    angularSpringConstant=500
    bAngularSpring=true
    bGravityOverrideInSimSpace=true
    bOverrideAngularDamping=true
    simulationSpace=BoneRelative
    alphaBoolBlend {blendOption=Linear}
    boundBone {boneName=dyn_coat_fr_r_04}
    componentAppliedLinearAccClamp=<3500 3500 3500>
    componentLinearAccScale=<0.5 0.5 0.7>
    componentLinearVelScale=<0.5 0.5 0.7>
    physicsBodyDefinitions [{
      sphereCollisionRadius=0
      boundBone {boneName=dyn_coat_fr_r_04}
      boxExtents=<15 20 20>
      constraintSetup {angularLimitsMax=<1 8 15>, angularLimitsMin=<-1 -45 -15>, angularTarget=<0 0 0>}
      localJointOffset=<5 0 0>
     }]
    relativeSpaceBone {boneName=root}
   }
   n42 = Anim Dynamics - Bone: dyn_coat_fr_r_05(ComponentPose=@n41, Alpha=1, ExternalForce="0,0,0") {
    angularDampingOverride=0.9999
    angularSpringConstant=500
    bAngularSpring=true
    bGravityOverrideInSimSpace=true
    bOverrideAngularDamping=true
    simulationSpace=BoneRelative
    alphaBoolBlend {blendOption=Linear}
    boundBone {boneName=dyn_coat_fr_r_05}
    componentAppliedLinearAccClamp=<3500 3500 3500>
    componentLinearAccScale=<0.5 0.5 0.7>
    componentLinearVelScale=<0.5 0.5 0.7>
    physicsBodyDefinitions [{
      sphereCollisionRadius=0
      boundBone {boneName=dyn_coat_fr_r_05}
      boxExtents=<15 20 20>
      constraintSetup {angularLimitsMax=<1 8 15>, angularLimitsMin=<-1 -45 -15>, angularTarget=<0 0 0>}
      localJointOffset=<5 0 0>
     }]
    relativeSpaceBone {boneName=root}
   }
   n43 = Anim Dynamics - Bone: dyn_coat_fr_r_06(ComponentPose=@n42, Alpha=1, ExternalForce="0,0,0") {
    angularDampingOverride=0.9999
    angularSpringConstant=500
    bAngularSpring=true
    bGravityOverrideInSimSpace=true
    bOverrideAngularDamping=true
    simulationSpace=BoneRelative
    alphaBoolBlend {blendOption=Linear}
    boundBone {boneName=dyn_coat_fr_r_06}
    componentAppliedLinearAccClamp=<3500 3500 3500>
    componentLinearAccScale=<0.5 0.5 0.7>
    componentLinearVelScale=<0.5 0.5 0.7>
    physicsBodyDefinitions [{
      sphereCollisionRadius=0
      boundBone {boneName=dyn_coat_fr_r_06}
      boxExtents=<15 20 20>
      constraintSetup {angularLimitsMax=<1 8 15>, angularLimitsMin=<-1 -45 -15>, angularTarget=<0 0 0>}
      localJointOffset=<5 0 0>
     }]
    relativeSpaceBone {boneName=root}
   }
   n44 = Anim Dynamics - Bone: dyn_coat_r_02(ComponentPose=@n43, Alpha=1, ExternalForce="0,0,0") {
    angularDampingOverride=0.9999
    angularSpringConstant=500
    bAngularSpring=true
    bGravityOverrideInSimSpace=true
    bOverrideAngularDamping=true
    simulationSpace=BoneRelative
    alphaBoolBlend {blendOption=Linear}
    boundBone {boneName=dyn_coat_r_02}
    componentAppliedLinearAccClamp=<5000 5000 5000>
    componentLinearAccScale=<0.5 0.5 0.7>
    componentLinearVelScale=<0.5 0.5 0.7>
    physicsBodyDefinitions [{
      sphereCollisionRadius=0
      boundBone {boneName=dyn_coat_r_02}
      boxExtents=<15 20 20>
      constraintSetup {angularLimitsMax=<1 3 15>, angularLimitsMin=<-1 -20 -15>, angularTarget=<0 0 0>}
      localJointOffset=<5 0 0>
     }]
    relativeSpaceBone {boneName=root}
   }
   n45 = Anim Dynamics - Bone: dyn_coat_r_03(ComponentPose=@n44, Alpha=1, ExternalForce="0,0,0") {
    angularDampingOverride=0.9999
    angularSpringConstant=500
    bAngularSpring=true
    bGravityOverrideInSimSpace=true
    bOverrideAngularDamping=true
    simulationSpace=BoneRelative
    alphaBoolBlend {blendOption=Linear}
    boundBone {boneName=dyn_coat_r_03}
    componentAppliedLinearAccClamp=<5000 5000 5000>
    componentLinearAccScale=<0.5 0.5 0.7>
    componentLinearVelScale=<0.5 0.5 0.7>
    physicsBodyDefinitions [{
      sphereCollisionRadius=0
      boundBone {boneName=dyn_coat_r_03}
      boxExtents=<15 20 20>
      constraintSetup {angularLimitsMax=<1 8 15>, angularLimitsMin=<-1 -45 -15>, angularTarget=<0 0 0>}
      localJointOffset=<5 0 0>
     }]
    relativeSpaceBone {boneName=root}
   }
   n46 = Anim Dynamics - Bone: dyn_coat_r_04(ComponentPose=@n45, Alpha=1, ExternalForce="0,0,0") {
    angularDampingOverride=0.9999
    angularSpringConstant=500
    bAngularSpring=true
    bGravityOverrideInSimSpace=true
    bOverrideAngularDamping=true
    simulationSpace=BoneRelative
    alphaBoolBlend {blendOption=Linear}
    boundBone {boneName=dyn_coat_r_04}
    componentAppliedLinearAccClamp=<3500 3500 3500>
    componentLinearAccScale=<0.5 0.5 0.7>
    componentLinearVelScale=<0.5 0.5 0.7>
    physicsBodyDefinitions [{
      sphereCollisionRadius=0
      boundBone {boneName=dyn_coat_r_04}
      boxExtents=<15 20 20>
      constraintSetup {angularLimitsMax=<1 8 15>, angularLimitsMin=<-1 -45 -15>, angularTarget=<0 0 0>}
      localJointOffset=<5 0 0>
     }]
    relativeSpaceBone {boneName=root}
   }
   n47 = Anim Dynamics - Bone: dyn_coat_r_05(ComponentPose=@n46, Alpha=1, ExternalForce="0,0,0") {
    angularDampingOverride=0.9999
    angularSpringConstant=500
    bAngularSpring=true
    bGravityOverrideInSimSpace=true
    bOverrideAngularDamping=true
    simulationSpace=BoneRelative
    alphaBoolBlend {blendOption=Linear}
    boundBone {boneName=dyn_coat_r_05}
    componentAppliedLinearAccClamp=<3500 3500 3500>
    componentLinearAccScale=<0.5 0.5 0.7>
    componentLinearVelScale=<0.5 0.5 0.7>
    physicsBodyDefinitions [{
      sphereCollisionRadius=0
      boundBone {boneName=dyn_coat_r_05}
      boxExtents=<15 20 20>
      constraintSetup {angularLimitsMax=<1 8 15>, angularLimitsMin=<-1 -45 -15>, angularTarget=<0 0 0>}
      localJointOffset=<5 0 0>
     }]
    relativeSpaceBone {boneName=root}
   }
   n48 = Anim Dynamics - Bone: dyn_coat_r_06(ComponentPose=@n47, Alpha=1, ExternalForce="0,0,0") {
    angularDampingOverride=0.9999
    angularSpringConstant=500
    bAngularSpring=true
    bGravityOverrideInSimSpace=true
    bOverrideAngularDamping=true
    simulationSpace=BoneRelative
    alphaBoolBlend {blendOption=Linear}
    boundBone {boneName=dyn_coat_r_06}
    componentAppliedLinearAccClamp=<3500 3500 3500>
    componentLinearAccScale=<0.5 0.5 0.7>
    componentLinearVelScale=<0.5 0.5 0.7>
    physicsBodyDefinitions [{
      sphereCollisionRadius=0
      boundBone {boneName=dyn_coat_r_06}
      boxExtents=<15 20 20>
      constraintSetup {angularLimitsMax=<1 8 15>, angularLimitsMin=<-1 -45 -15>, angularTarget=<0 0 0>}
      localJointOffset=<5 0 0>
     }]
    relativeSpaceBone {boneName=root}
   }
   n49 = Anim Dynamics - Bone: dyn_coat_bk_r_02(ComponentPose=@n48, Alpha=1, ExternalForce="0,0,0") {
    angularDampingOverride=0.9999
    angularSpringConstant=500
    bAngularSpring=true
    bGravityOverrideInSimSpace=true
    bOverrideAngularDamping=true
    simulationSpace=BoneRelative
    alphaBoolBlend {blendOption=Linear}
    boundBone {boneName=dyn_coat_bk_r_02}
    componentAppliedLinearAccClamp=<5000 5000 5000>
    componentLinearAccScale=<0.5 0.5 0.7>
    componentLinearVelScale=<0.5 0.5 0.7>
    physicsBodyDefinitions [{
      sphereCollisionRadius=0
      boundBone {boneName=dyn_coat_bk_r_02}
      boxExtents=<15 20 20>
      constraintSetup {angularLimitsMax=<1 3 15>, angularLimitsMin=<-1 -20 -15>, angularTarget=<0 0 0>}
      localJointOffset=<5 0 0>
     }]
    relativeSpaceBone {boneName=root}
   }
   n50 = Anim Dynamics - Bone: dyn_coat_bk_r_03(ComponentPose=@n49, Alpha=1, ExternalForce="0,0,0") {
    angularDampingOverride=0.9999
    angularSpringConstant=500
    bAngularSpring=true
    bGravityOverrideInSimSpace=true
    bOverrideAngularDamping=true
    simulationSpace=BoneRelative
    alphaBoolBlend {blendOption=Linear}
    boundBone {boneName=dyn_coat_bk_r_03}
    componentAppliedLinearAccClamp=<5000 5000 5000>
    componentLinearAccScale=<0.5 0.5 0.7>
    componentLinearVelScale=<0.5 0.5 0.7>
    physicsBodyDefinitions [{
      sphereCollisionRadius=0
      boundBone {boneName=dyn_coat_bk_r_03}
      boxExtents=<15 20 20>
      constraintSetup {angularLimitsMax=<1 8 15>, angularLimitsMin=<-1 -45 -15>, angularTarget=<0 0 0>}
      localJointOffset=<5 0 0>
     }]
    relativeSpaceBone {boneName=root}
   }
   n51 = Anim Dynamics - Bone: dyn_coat_bk_r_04(ComponentPose=@n50, Alpha=1, ExternalForce="0,0,0") {
    angularDampingOverride=0.9999
    angularSpringConstant=500
    bAngularSpring=true
    bGravityOverrideInSimSpace=true
    bOverrideAngularDamping=true
    simulationSpace=BoneRelative
    alphaBoolBlend {blendOption=Linear}
    boundBone {boneName=dyn_coat_bk_r_04}
    componentAppliedLinearAccClamp=<3500 3500 3500>
    componentLinearAccScale=<0.5 0.5 0.7>
    componentLinearVelScale=<0.5 0.5 0.7>
    physicsBodyDefinitions [{
      sphereCollisionRadius=0
      boundBone {boneName=dyn_coat_bk_r_04}
      boxExtents=<15 20 20>
      constraintSetup {angularLimitsMax=<1 8 15>, angularLimitsMin=<-1 -45 -15>, angularTarget=<0 0 0>}
      localJointOffset=<5 0 0>
     }]
    relativeSpaceBone {boneName=root}
   }
   n52 = Anim Dynamics - Bone: dyn_coat_bk_r_05(ComponentPose=@n51, Alpha=1, ExternalForce="0,0,0") {
    angularDampingOverride=0.9999
    angularSpringConstant=500
    bAngularSpring=true
    bGravityOverrideInSimSpace=true
    bOverrideAngularDamping=true
    simulationSpace=BoneRelative
    alphaBoolBlend {blendOption=Linear}
    boundBone {boneName=dyn_coat_bk_r_05}
    componentAppliedLinearAccClamp=<3500 3500 3500>
    componentLinearAccScale=<0.5 0.5 0.7>
    componentLinearVelScale=<0.5 0.5 0.7>
    physicsBodyDefinitions [{
      sphereCollisionRadius=0
      boundBone {boneName=dyn_coat_bk_r_05}
      boxExtents=<15 20 20>
      constraintSetup {angularLimitsMax=<1 8 15>, angularLimitsMin=<-1 -45 -15>, angularTarget=<0 0 0>}
      localJointOffset=<5 0 0>
     }]
    relativeSpaceBone {boneName=root}
   }
   n53 = Anim Dynamics - Bone: dyn_coat_bk_r_06(ComponentPose=@n52, Alpha=1, ExternalForce="0,0,0") {
    angularDampingOverride=0.9999
    angularSpringConstant=500
    bAngularSpring=true
    bGravityOverrideInSimSpace=true
    bOverrideAngularDamping=true
    simulationSpace=BoneRelative
    alphaBoolBlend {blendOption=Linear}
    boundBone {boneName=dyn_coat_bk_r_06}
    componentAppliedLinearAccClamp=<3500 3500 3500>
    componentLinearAccScale=<0.5 0.5 0.7>
    componentLinearVelScale=<0.5 0.5 0.7>
    physicsBodyDefinitions [{
      sphereCollisionRadius=0
      boundBone {boneName=dyn_coat_bk_r_06}
      boxExtents=<15 20 20>
      constraintSetup {angularLimitsMax=<1 8 15>, angularLimitsMin=<-1 -45 -15>, angularTarget=<0 0 0>}
      localJointOffset=<5 0 0>
     }]
    relativeSpaceBone {boneName=root}
   }
   n54 = Anim Dynamics - Bone: dyn_coat_bk_01(ComponentPose=@n53, Alpha=1, ExternalForce="0,0,0") {
    angularDampingOverride=0.9999
    angularSpringConstant=500
    bAngularSpring=true
    bGravityOverrideInSimSpace=true
    bOverrideAngularDamping=true
    simulationSpace=BoneRelative
    alphaBoolBlend {blendOption=Linear}
    boundBone {boneName=dyn_coat_bk_01}
    componentAppliedLinearAccClamp=<3500 3500 3500>
    componentLinearAccScale=<1.2 1.2 1.2>
    componentLinearVelScale=<1.2 1.2 1.2>
    physicsBodyDefinitions [{
      sphereCollisionRadius=0
      boundBone {boneName=dyn_coat_bk_01}
      boxExtents=<15 20 20>
      constraintSetup {angularLimitsMax=<1 8 20>, angularLimitsMin=<-1 -45 -25>, angularTarget=<0 0 0>}
      localJointOffset=<5 0 0>
     }]
    relativeSpaceBone {boneName=root}
   }
   n55 = Anim Dynamics - Bone: dyn_coat_bk_02(ComponentPose=@n54, Alpha=1, ExternalForce="0,0,0") {
    angularDampingOverride=0.9999
    angularSpringConstant=500
    bAngularSpring=true
    bGravityOverrideInSimSpace=true
    bOverrideAngularDamping=true
    simulationSpace=BoneRelative
    alphaBoolBlend {blendOption=Linear}
    boundBone {boneName=dyn_coat_bk_02}
    componentAppliedLinearAccClamp=<3500 3500 3500>
    componentLinearAccScale=<1.2 1.2 1.2>
    componentLinearVelScale=<1.2 1.2 1.2>
    physicsBodyDefinitions [{
      sphereCollisionRadius=0
      boundBone {boneName=dyn_coat_bk_02}
      boxExtents=<15 20 20>
      constraintSetup {angularLimitsMax=<1 8 20>, angularLimitsMin=<-1 -45 -25>, angularTarget=<0 0 0>}
      localJointOffset=<5 0 0>
     }]
    relativeSpaceBone {boneName=root}
   }
   n56 = Anim Dynamics - Bone: dyn_coat_bk_03(ComponentPose=@n55, Alpha=1, ExternalForce="0,0,0") {
    angularDampingOverride=0.9999
    angularSpringConstant=500
    bAngularSpring=true
    bGravityOverrideInSimSpace=true
    bOverrideAngularDamping=true
    simulationSpace=BoneRelative
    alphaBoolBlend {blendOption=Linear}
    boundBone {boneName=dyn_coat_bk_03}
    componentAppliedLinearAccClamp=<3500 3500 3500>
    componentLinearAccScale=<1.2 1.2 1.2>
    componentLinearVelScale=<1.2 1.2 1.2>
    physicsBodyDefinitions [{
      sphereCollisionRadius=0
      boundBone {boneName=dyn_coat_bk_03}
      boxExtents=<15 20 20>
      constraintSetup {angularLimitsMax=<1 8 20>, angularLimitsMin=<-1 -45 -25>, angularTarget=<0 0 0>}
      localJointOffset=<5 0 0>
     }]
    relativeSpaceBone {boneName=root}
   }
   n57 = Anim Dynamics - Bone: dyn_coat_bkend_l_01(ComponentPose=@n56, Alpha=1, ExternalForce="0,0,0") {
    angularDampingOverride=0.9999
    angularSpringConstant=500
    bAngularSpring=true
    bGravityOverrideInSimSpace=true
    bOverrideAngularDamping=true
    simulationSpace=BoneRelative
    alphaBoolBlend {blendOption=Linear}
    boundBone {boneName=dyn_coat_bkend_l_01}
    componentAppliedLinearAccClamp=<3500 3500 3500>
    componentLinearAccScale=<0.5 0.5 0.7>
    componentLinearVelScale=<0.5 0.5 0.7>
    physicsBodyDefinitions [{
      sphereCollisionRadius=0
      boundBone {boneName=dyn_coat_bkend_l_01}
      boxExtents=<15 20 20>
      constraintSetup {angularLimitsMax=<1 8 20>, angularLimitsMin=<-1 -45 -25>, angularTarget=<0 0 0>}
      localJointOffset=<5 0 0>
     }]
    relativeSpaceBone {boneName=root}
   }
   n58 = Anim Dynamics - Bone: dyn_coat_bkend_l_02(ComponentPose=@n57, Alpha=1, ExternalForce="0,0,0") {
    angularDampingOverride=0.9999
    angularSpringConstant=500
    bAngularSpring=true
    bGravityOverrideInSimSpace=true
    bOverrideAngularDamping=true
    simulationSpace=BoneRelative
    alphaBoolBlend {blendOption=Linear}
    boundBone {boneName=dyn_coat_bkend_l_02}
    componentAppliedLinearAccClamp=<3500 3500 3500>
    componentLinearAccScale=<0.5 0.5 0.7>
    componentLinearVelScale=<0.5 0.5 0.7>
    physicsBodyDefinitions [{
      sphereCollisionRadius=0
      boundBone {boneName=dyn_coat_bkend_l_02}
      boxExtents=<15 20 20>
      constraintSetup {angularLimitsMax=<1 8 20>, angularLimitsMin=<-1 -45 -25>, angularTarget=<0 0 0>}
      localJointOffset=<5 0 0>
     }]
    relativeSpaceBone {boneName=root}
   }
   n59 = Anim Dynamics - Bone: dyn_coat_bkend_l_03(ComponentPose=@n58, Alpha=1, ExternalForce="0,0,0") {
    angularDampingOverride=0.9999
    angularSpringConstant=500
    bAngularSpring=true
    bGravityOverrideInSimSpace=true
    bOverrideAngularDamping=true
    simulationSpace=BoneRelative
    alphaBoolBlend {blendOption=Linear}
    boundBone {boneName=dyn_coat_bkend_l_03}
    componentAppliedLinearAccClamp=<3500 3500 3500>
    componentLinearAccScale=<0.5 0.5 0.7>
    componentLinearVelScale=<0.5 0.5 0.7>
    physicsBodyDefinitions [{
      sphereCollisionRadius=0
      boundBone {boneName=dyn_coat_bkend_l_03}
      boxExtents=<15 20 20>
      constraintSetup {angularLimitsMax=<1 8 20>, angularLimitsMin=<-1 -45 -25>, angularTarget=<0 0 0>}
      localJointOffset=<5 0 0>
     }]
    relativeSpaceBone {boneName=root}
   }
   n60 = Anim Dynamics - Bone: dyn_coat_bkend_r_01(ComponentPose=@n59, Alpha=1, ExternalForce="0,0,0") {
    angularDampingOverride=0.9999
    angularSpringConstant=500
    bAngularSpring=true
    bGravityOverrideInSimSpace=true
    bOverrideAngularDamping=true
    simulationSpace=BoneRelative
    alphaBoolBlend {blendOption=Linear}
    boundBone {boneName=dyn_coat_bkend_r_01}
    componentAppliedLinearAccClamp=<3500 3500 3500>
    componentLinearAccScale=<0.5 0.5 0.7>
    componentLinearVelScale=<0.5 0.5 0.7>
    physicsBodyDefinitions [{
      sphereCollisionRadius=0
      boundBone {boneName=dyn_coat_bkend_r_01}
      boxExtents=<15 20 20>
      constraintSetup {angularLimitsMax=<1 8 20>, angularLimitsMin=<-1 -45 -25>, angularTarget=<0 0 0>}
      localJointOffset=<5 0 0>
     }]
    relativeSpaceBone {boneName=root}
   }
   n61 = Anim Dynamics - Bone: dyn_coat_bkend_r_02(ComponentPose=@n60, Alpha=1, ExternalForce="0,0,0") {
    angularDampingOverride=0.9999
    angularSpringConstant=500
    bAngularSpring=true
    bGravityOverrideInSimSpace=true
    bOverrideAngularDamping=true
    simulationSpace=BoneRelative
    alphaBoolBlend {blendOption=Linear}
    boundBone {boneName=dyn_coat_bkend_r_02}
    componentAppliedLinearAccClamp=<3500 3500 3500>
    componentLinearAccScale=<0.5 0.5 0.7>
    componentLinearVelScale=<0.5 0.5 0.7>
    physicsBodyDefinitions [{
      sphereCollisionRadius=0
      boundBone {boneName=dyn_coat_bkend_r_02}
      boxExtents=<15 20 20>
      constraintSetup {angularLimitsMax=<1 8 20>, angularLimitsMin=<-1 -45 -25>, angularTarget=<0 0 0>}
      localJointOffset=<5 0 0>
     }]
    relativeSpaceBone {boneName=root}
   }
   n62 = Anim Dynamics - Bone: dyn_coat_bkend_r_03(ComponentPose=@n61, Alpha=1, ExternalForce="0,0,0") {
    angularDampingOverride=0.9999
    angularSpringConstant=500
    bAngularSpring=true
    bGravityOverrideInSimSpace=true
    bOverrideAngularDamping=true
    simulationSpace=BoneRelative
    alphaBoolBlend {blendOption=Linear}
    boundBone {boneName=dyn_coat_bkend_r_03}
    componentAppliedLinearAccClamp=<3500 3500 3500>
    componentLinearAccScale=<0.5 0.5 0.7>
    componentLinearVelScale=<0.5 0.5 0.7>
    physicsBodyDefinitions [{
      sphereCollisionRadius=0
      boundBone {boneName=dyn_coat_bkend_r_03}
      boxExtents=<15 20 20>
      constraintSetup {angularLimitsMax=<1 8 20>, angularLimitsMin=<-1 -45 -25>, angularTarget=<0 0 0>}
      localJointOffset=<5 0 0>
     }]
    relativeSpaceBone {boneName=root}
   }
   n63 = RigidBody!(
    ComponentPose=Local To Component(LocalPose=Use cached pose 'initial Pose'())
    Alpha=Get InVehicleRigidAlpha() -> (InVehicleRigidAlpha:float)
    ExternalForce
   ) -> (Pose) {
    bClampLinearTranslationLimitToRefPose=true
    bOverrideWorldGravity=true
    overridePhysicsAsset=&/M_MED_Racer_Zero_Mask_Dyn
    simulationSpace=BaseBoneSpace
    alphaBoolBlend {blendOption=Linear}
    baseBoneRef {boneName=spine_03}
    componentAppliedLinearAccClamp=<3000 3000 3000>
    componentLinearAccScale=<0.4 0.4 0.4>
    componentLinearVelScale=<0.4 0.4 0.4>
    overrideWorldGravity {z=-980}
   }
   n64 = RigidBody(ComponentPose=@n63) {
    bClampLinearTranslationLimitToRefPose=true
    bOverrideWorldGravity=true
    overridePhysicsAsset=&/M_MED_Racer_Zero_Mask_zip_Dyn
    simulationSpace=BaseBoneSpace
    alphaBoolBlend {blendOption=Linear}
    baseBoneRef {boneName=spine_05}
    componentAppliedLinearAccClamp=<3000 3000 3000>
    componentLinearAccScale=<0.4 0.4 0.4>
    componentLinearVelScale=<0.4 0.4 0.4>
    overrideWorldGravity {z=-980}
   }
   n65 = Blend Poses by bool(BlendPose_0=Component To Local(ComponentPose=@n64), BlendPose_1=Component To Local(ComponentPose=@n62), bActiveValue=Get bRigidBodyEnabled()) {blendType=Linear}
   n66 = RigidBody(ComponentPose=Local To Component(LocalPose=Use cached pose 'initial Pose'())) {
    bClampLinearTranslationLimitToRefPose=true
    bOverrideWorldGravity=true
    overridePhysicsAsset=&M_MED_Racer_Zero_Mask_Dyn
    simulationSpace=BaseBoneSpace
    alphaBoolBlend {blendOption=Linear}
    baseBoneRef {boneName=root}
    componentAppliedLinearAccClamp=<3000 3000 3000>
    componentLinearAccScale=<0.2 0.2 0.2>
    componentLinearVelScale=<0.2 0.2 0.2>
    overrideWorldGravity {z=-800}
   }
   n67 = "Constraint - dyn_coat_bk_01 (P)"(ComponentPose=@n66, Alpha=0.5, ConstraintWeights_0=1) {
    alphaBoolBlend {blendOption=Linear}
    boneToModify {boneName=dyn_coat_bk_01}
    constraintSetup [{targetBone {boneName=spine_02}}]
    constraintWeights [1]
   }
   n68 = "Constraint - dyn_coat_bkend_l_01 (P)"(ComponentPose=@n67, Alpha=0.5, ConstraintWeights_0=1) {
    alphaBoolBlend {blendOption=Linear}
    boneToModify {boneName=dyn_coat_bkend_l_01}
    constraintSetup [{targetBone {boneName=thigh_l}}]
    constraintWeights [1]
   }
   n69 = "Constraint - dyn_coat_bkend_r_01 (P)"(ComponentPose=@n68, Alpha=0.5, ConstraintWeights_0=1) {
    alphaBoolBlend {blendOption=Linear}
    boneToModify {boneName=dyn_coat_bkend_r_01}
    constraintSetup [{targetBone {boneName=thigh_r}}]
    constraintWeights [1]
   }
   n70 = "Constraint - dyn_coat_bk_l_03 (P)"(ComponentPose=@n69, Alpha=0.5, ConstraintWeights_0=1) {
    alphaBoolBlend {blendOption=Linear}
    boneToModify {boneName=dyn_coat_bk_l_03}
    constraintSetup [{targetBone {boneName=thigh_l}}]
    constraintWeights [1]
   }
   n71 = "Constraint - dyn_coat_bk_r_03 (P)"(ComponentPose=@n70, Alpha=0.5, ConstraintWeights_0=1) {
    alphaBoolBlend {blendOption=Linear}
    boneToModify {boneName=dyn_coat_bk_r_03}
    constraintSetup [{targetBone {boneName=thigh_r}}]
    constraintWeights [1]
   }
   n72 = "Transform (Modify) Bone - Bone: dyn_coat_bk_03"!(ComponentPose=@n71, Alpha=1, Translation="15,0,-5", Rotation="20,0,0") -> (Pose) {
    alpha=0
    rotationMode=BMM_Additive
    rotationSpace=BCS_BoneSpace
    translationMode=BMM_Additive
    translationSpace=BCS_BoneSpace
    alphaBoolBlend {blendOption=Linear}
    alphaScaleBiasClamp {bInterpResult=true, interpSpeedIncreasing=20}
    boneToModify {boneName=dyn_coat_bk_03}
    rotation {pitch=20}
    translation {x=15, z=-5}
   }
   n73 = "Transform (Modify) Bone - Bone: dyn_coat_bkend_l_02"!(ComponentPose=@n72, Alpha=1, Translation="0,0,5", Rotation="15.000,0.000,15.000") -> (Pose) {
    alpha=0
    rotationMode=BMM_Additive
    rotationSpace=BCS_BoneSpace
    translationMode=BMM_Replace
    translationSpace=BCS_BoneSpace
    alphaBoolBlend {blendOption=Linear}
    alphaScaleBiasClamp {bInterpResult=true, interpSpeedIncreasing=20}
    boneToModify {boneName=dyn_coat_bkend_l_02}
    rotation=<rot 15 0 15>
    translation {z=5}
   }
   n74 = "Transform (Modify) Bone - Bone: dyn_coat_bkend_r_02"!(ComponentPose=@n73, Alpha=1, Translation="0,0,3", Rotation="15.000,-0.000,-15.000") -> (Pose) {
    alpha=0
    rotationMode=BMM_Additive
    rotationSpace=BCS_BoneSpace
    translationMode=BMM_Additive
    translationSpace=BCS_BoneSpace
    alphaBoolBlend {blendOption=Linear}
    alphaScaleBiasClamp {bInterpResult=true, interpSpeedIncreasing=20}
    boneToModify {boneName=dyn_coat_bkend_r_02}
    rotation=<rot 15 0 -15>
    translation {z=3}
   }
   n75 = "Transform (Modify) Bone - Bone: dyn_coat_bk_l_03"!(ComponentPose=@n74, Alpha=1, Translation="0,-5,5", Rotation="0,0,0") -> (Pose) {
    alpha=0
    rotationMode=BMM_Additive
    rotationSpace=BCS_BoneSpace
    translationMode=BMM_Replace
    translationSpace=BCS_BoneSpace
    alphaBoolBlend {blendOption=Linear}
    alphaScaleBiasClamp {bInterpResult=true, interpSpeedIncreasing=20}
    boneToModify {boneName=dyn_coat_bk_l_03}
    translation {y=-5, z=5}
   }
   n76 = "Transform (Modify) Bone - Bone: dyn_coat_bk_r_03"(ComponentPose=@n75, Alpha=1, Translation="0,-5,5", Rotation="0,0,0") {
    alpha=0
    rotationMode=BMM_Additive
    rotationSpace=BCS_BoneSpace
    translationMode=BMM_Replace
    translationSpace=BCS_BoneSpace
    alphaBoolBlend {blendOption=Linear}
    alphaScaleBiasClamp {bInterpResult=true, interpSpeedIncreasing=20}
    boneToModify {boneName=dyn_coat_bk_r_03}
    translation {y=-5, z=5}
   }
   n77 = Get DBNO_Crawl() -> (DBNO_Crawl:double)
   n78 = "Transform (Modify) Bone - Bone: dyn_coat_bk_03"(ComponentPose=@n76, Alpha=@n77, Translation="8,0,0", Rotation="0,0,0") {
    alpha=0
    rotationMode=BMM_Additive
    rotationSpace=BCS_BoneSpace
    translationMode=BMM_Replace
    translationSpace=BCS_BoneSpace
    alphaBoolBlend {blendOption=Linear}
    alphaScaleBiasClamp {bInterpResult=true, interpSpeedDecreasing=5, interpSpeedIncreasing=7}
    boneToModify {boneName=dyn_coat_bk_03}
    translation {x=8}
   }
   n79 = "Transform (Modify) Bone - Bone: dyn_coat_bk_l_03"(ComponentPose=@n78, Alpha=@n77, Translation="0,0,4", Rotation="0,0,0") {
    alpha=0
    rotationMode=BMM_Additive
    rotationSpace=BCS_BoneSpace
    translationMode=BMM_Replace
    translationSpace=BCS_BoneSpace
    alphaBoolBlend {blendOption=Linear}
    alphaScaleBiasClamp {bInterpResult=true, interpSpeedDecreasing=5, interpSpeedIncreasing=7}
    boneToModify {boneName=dyn_coat_bk_l_03}
    translation {z=4}
   }
   n80 = "Transform (Modify) Bone - Bone: dyn_coat_bk_03"(ComponentPose=@n79, Alpha=1, Translation="-10,-6,0", Rotation="0,0,15.000") {
    alpha=0
    rotationMode=BMM_Additive
    rotationSpace=BCS_BoneSpace
    translationMode=BMM_Replace
    translationSpace=BCS_BoneSpace
    alphaBoolBlend {blendOption=Linear}
    alphaScaleBiasClamp {bInterpResult=true, interpSpeedIncreasing=20}
    boneToModify {boneName=dyn_coat_bk_03}
    rotation {roll=15}
    translation=<-10 -6>
   }
   n81 = "Constraint - dyn_coat_bkend_r_02 (P)"(ComponentPose=@n80, Alpha=1, ConstraintWeights_0=1) {
    alphaBoolBlend {blendOption=Linear}
    alphaScaleBiasClamp {bClampResult=true, bInterpResult=true, clampMax=0.6, interpSpeedDecreasing=5, interpSpeedIncreasing=7}
    boneToModify {boneName=dyn_coat_bkend_r_02}
    constraintSetup [{targetBone {boneName=thigh_r}}]
    constraintWeights [1]
   }
   n82 = "Constraint - dyn_coat_bk_r_03 (P)"(ComponentPose=@n81, Alpha=1, ConstraintWeights_0=1) {
    alphaBoolBlend {blendOption=Linear}
    alphaScaleBiasClamp {bClampResult=true, bInterpResult=true, clampMax=0.8, interpSpeedDecreasing=5, interpSpeedIncreasing=7}
    boneToModify {boneName=dyn_coat_bk_r_03}
    constraintSetup [{targetBone {boneName=thigh_r}}]
    constraintWeights [1]
   }
   n83 = "Constraint - dyn_coat_bk_r_04 (P)"(ComponentPose=@n82, Alpha=1, ConstraintWeights_0=1) {
    alphaBoolBlend {blendOption=Linear}
    alphaScaleBiasClamp {bClampResult=true, bInterpResult=true, clampMax=0.8, interpSpeedDecreasing=5, interpSpeedIncreasing=7}
    boneToModify {boneName=dyn_coat_bk_r_04}
    constraintSetup [{targetBone {boneName=thigh_r}}]
    constraintWeights [1]
   }
   n84 = "Constraint - dyn_coat_bk_r_05 (P)"(ComponentPose=@n83, Alpha=1, ConstraintWeights_0=1) {
    alphaBoolBlend {blendOption=Linear}
    alphaScaleBiasClamp {bClampResult=true, bInterpResult=true, clampMax=0.8, interpSpeedDecreasing=5, interpSpeedIncreasing=7}
    boneToModify {boneName=dyn_coat_bk_r_05}
    constraintSetup [{targetBone {boneName=thigh_r}}]
    constraintWeights [1]
   }
   n85 = "Constraint - dyn_coat_bk_r_06 (P)"(ComponentPose=@n84, Alpha=1, ConstraintWeights_0=1) {
    alphaBoolBlend {blendOption=Linear}
    alphaScaleBiasClamp {bClampResult=true, bInterpResult=true, clampMax=0.8, interpSpeedDecreasing=5, interpSpeedIncreasing=7}
    boneToModify {boneName=dyn_coat_bk_r_06}
    constraintSetup [{targetBone {boneName=thigh_r}}]
    constraintWeights [1]
   }
   n86 = "Transform (Modify) Bone - Bone: dyn_coat_bk_r_04"!(ComponentPose=@n85, Alpha=1, Scale="0.7,1.000,1.000") -> (Pose) {
    alpha=0
    rotationSpace=BCS_BoneSpace
    scaleMode=BMM_Replace
    scaleSpace=BCS_BoneSpace
    translationSpace=BCS_BoneSpace
    alphaBoolBlend {blendOption=Linear}
    alphaScaleBiasClamp {bInterpResult=true, interpSpeedIncreasing=20}
    boneToModify {boneName=dyn_coat_bk_r_04}
    scale {x=0.7}
   }
   n87 = "Constraint - dyn_coat_r_04 (P)"(ComponentPose=@n86, Alpha=1, ConstraintWeights_0=1) {
    alphaBoolBlend {blendOption=Linear}
    alphaScaleBiasClamp {bClampResult=true, bInterpResult=true, clampMax=0.8, interpSpeedDecreasing=5, interpSpeedIncreasing=7}
    boneToModify {boneName=dyn_coat_r_04}
    constraintSetup [{targetBone {boneName=thigh_r}}]
    constraintWeights [1]
   }
   n88 = "Constraint - dyn_coat_fr_r_04 (P)"(ComponentPose=@n87, Alpha=1, ConstraintWeights_0=1) {
    alphaBoolBlend {blendOption=Linear}
    alphaScaleBiasClamp {bClampResult=true, bInterpResult=true, clampMax=0.8, interpSpeedDecreasing=5, interpSpeedIncreasing=7}
    boneToModify {boneName=dyn_coat_fr_r_04}
    constraintSetup [{targetBone {boneName=thigh_r}}]
    constraintWeights [1]
   }
   n89 = Component To Local(ComponentPose=@n88)
   n90 = Get bIsDBNO()
   n91 = "Transform (Modify) Bone - Bone: dyn_coat_bk_01"(
    ComponentPose=Local To Component(LocalPose=Blend Poses by bool(BlendPose_0=@n89, BlendPose_1=@n65, bActiveValue=@n90) {blendType=Linear})
    Alpha=1
    Translation="0,0,0"
    Rotation="3,0,0"
    Scale="1,1,1"
   ) {
    alpha=0
    rotationMode=BMM_Additive
    rotationSpace=BCS_BoneSpace
    alphaBoolBlend {blendOption=Linear}
    alphaScaleBiasClamp {bInterpResult=true, interpSpeedIncreasing=20}
    boneToModify {boneName=dyn_coat_bk_01}
    rotation {pitch=3}
   }
   n92 = "Transform (Modify) Bone - Bone: dyn_coat_bk_r_01"(ComponentPose=@n91, Alpha=1, Translation="0,0,0", Rotation="6,0,0", Scale="1,1,1") {
    alpha=0
    rotationMode=BMM_Additive
    rotationSpace=BCS_BoneSpace
    alphaBoolBlend {blendOption=Linear}
    alphaScaleBiasClamp {bInterpResult=true, interpSpeedIncreasing=20}
    boneToModify {boneName=dyn_coat_bk_r_01}
    rotation {pitch=6}
   }
   n93 = "Transform (Modify) Bone - Bone: dyn_coat_fr_l_02"!(ComponentPose=@n92, Alpha=1, Translation="-3.333,-5.056,-1.948", Rotation="23.917,2.073,-3.340") -> (Pose) {
    alpha=0
    rotationMode=BMM_Replace
    rotationSpace=BCS_BoneSpace
    translationMode=BMM_Replace
    translationSpace=BCS_BoneSpace
    alphaBoolBlend {blendOption=Linear}
    alphaScaleBiasClamp {bInterpResult=true, interpSpeedDecreasing=5, interpSpeedIncreasing=7}
    boneToModify {boneName=dyn_coat_fr_l_02}
    rotation=<rot 23.917 2.073 -3.34>
    translation=<-3.333 -5.056 -1.948>
   }
   n94 = "Transform (Modify) Bone - Bone: dyn_coat_fr_l_05"!(ComponentPose=@n93, Alpha=1, Rotation="-15.003,0.109,-0.291") -> (Pose) {
    alpha=0
    rotationMode=BMM_Additive
    rotationSpace=BCS_BoneSpace
    alphaBoolBlend {blendOption=Linear}
    alphaScaleBiasClamp {bInterpResult=true, interpSpeedDecreasing=5, interpSpeedIncreasing=7}
    boneToModify {boneName=dyn_coat_fr_l_05}
    rotation=<rot -15.003 0.109 -0.291>
   }
   n95 = Get Crouch walk() -> (Crouch walk:double)
   n96 = "Transform (Modify) Bone - Bone: dyn_coat_fr_r_MSTR"(ComponentPose=@n94, Alpha=@n95, Translation="-2,0.000,-0.5", Rotation="0,0,0", Scale="1,1,1") {
    alpha=0
    rotationMode=BMM_Additive
    rotationSpace=BCS_BoneSpace
    translationMode=BMM_Additive
    translationSpace=BCS_BoneSpace
    alphaBoolBlend {blendOption=Linear}
    alphaScaleBiasClamp {bInterpResult=true, interpSpeedIncreasing=20}
    boneToModify {boneName=dyn_coat_fr_r_MSTR}
    translation {x=-2, z=-0.5}
   }
   n97 = "Transform (Modify) Bone - Bone: dyn_coat_fr_r_02"(ComponentPose=@n96, Alpha=@n95, Translation="-6.5,0.000,0.000", Rotation="5.000,0.000,0.000", Scale="1,1,1") {
    alpha=0
    rotationMode=BMM_Additive
    rotationSpace=BCS_BoneSpace
    translationMode=BMM_Additive
    translationSpace=BCS_BoneSpace
    alphaBoolBlend {blendOption=Linear}
    alphaScaleBiasClamp {bInterpResult=true, interpSpeedIncreasing=20}
    boneToModify {boneName=dyn_coat_fr_r_02}
    rotation {pitch=5}
    translation {x=-6.5}
   }
   n98 = "Transform (Modify) Bone - Bone: dyn_coat_bk_03"!(ComponentPose=@n97, Alpha=@n95, Translation="0,0.000,-3", Rotation="-5.000,0.000,0.000", Scale:Vector="1,1,1") -> (Pose) {
    alpha=0
    rotationMode=BMM_Additive
    rotationSpace=BCS_BoneSpace
    translationMode=BMM_Additive
    translationSpace=BCS_BoneSpace
    alphaBoolBlend {blendOption=Linear}
    alphaScaleBiasClamp {bInterpResult=true, interpSpeedIncreasing=20}
    boneToModify {boneName=dyn_coat_bk_03}
    rotation {pitch=-5}
    translation {z=-3}
   }
   n99 = "Transform (Modify) Bone - Bone: dyn_coat_fr_l_MSTR"(ComponentPose=@n98, Alpha=@n95, Translation="0,0,0", Rotation="5.000,0.000,0.000", Scale="1,1,1") {
    alpha=0
    rotationMode=BMM_Additive
    rotationSpace=BCS_BoneSpace
    translationMode=BMM_Additive
    translationSpace=BCS_BoneSpace
    alphaBoolBlend {blendOption=Linear}
    alphaScaleBiasClamp {bInterpResult=true, interpSpeedIncreasing=20}
    boneToModify {boneName=dyn_coat_fr_l_MSTR}
    rotation {pitch=5}
   }
   n100 = Component To Local(ComponentPose=@n99)
   n101 = Get bIsSkyDiving()
   n102 = "Constraint - dyn_zipper_coat_MSTR (PPPPP)"!(
    ComponentPose=Local To Component(LocalPose=Blend Poses by bool(BlendPose_0=@n11, BlendPose_1=@n100, bActiveValue=@n101) {blendType=Linear})
    Alpha=1
    ConstraintWeights_0=0.0137
    ConstraintWeights_1:float=0.0495
    ConstraintWeights_2:float=0.8998
    ConstraintWeights_3:float=0.0249
    ConstraintWeights_4:float=0.0119
   ) -> (Pose) {
    alphaBoolBlend {blendOption=Linear}
    boneToModify {boneName=dyn_zipper_coat_MSTR}
    constraintSetup [
     {targetBone {boneName=dyn_coat_l_05}}
     {targetBone {boneName=dyn_coat_fr_l_04}}
     {targetBone {boneName=dyn_coat_fr_l_05}}
     {targetBone {boneName=dyn_coat_fr_l_06}}
     {targetBone {boneName=dyn_coat_l_04}}
    ]
    constraintWeights [1,1,1,1,1]
   }
   n103 = "Transform (Modify) Bone - Bone: dyn_fx_chest_trail_bk_l_MSTR"!(ComponentPose=@n102, Alpha=0.3, Rotation=Get sineRotation6() -> (sineRotation6:Rotator)) -> (Pose) {
    alpha=0
    rotationMode=BMM_Additive
    rotationSpace=BCS_BoneSpace
    translationMode=BMM_Additive
    translationSpace=BCS_BoneSpace
    alphaBoolBlend {blendOption=Linear}
    alphaScaleBiasClamp {bInterpResult=true, clampMax=0.6, interpSpeedIncreasing=20}
    boneToModify {boneName=dyn_fx_chest_trail_bk_l_MSTR}
   }
   n104 = "Transform (Modify) Bone - Bone: dyn_fx_chest_trail_bk_r_MSTR"!(ComponentPose=@n103, Alpha=0.3, Rotation=Get sineRotation1() -> (sineRotation1:Rotator)) -> (Pose) {
    alpha=0
    rotationMode=BMM_Additive
    rotationSpace=BCS_BoneSpace
    translationMode=BMM_Additive
    translationSpace=BCS_BoneSpace
    alphaBoolBlend {blendOption=Linear}
    alphaScaleBiasClamp {bInterpResult=true, clampMax=0.6, interpSpeedIncreasing=20}
    boneToModify {boneName=dyn_fx_chest_trail_bk_r_MSTR}
   }
   n105 = "Transform (Modify) Bone - Bone: dyn_fx_chest_trail_fr_l_MSTR"!(ComponentPose=@n104, Alpha=0.6, Rotation=Get sineRotation5() -> (sineRotation5:Rotator)) -> (Pose) {
    alpha=0
    rotationMode=BMM_Additive
    rotationSpace=BCS_BoneSpace
    translationMode=BMM_Additive
    translationSpace=BCS_BoneSpace
    alphaBoolBlend {blendOption=Linear}
    alphaScaleBiasClamp {bInterpResult=true, clampMax=0.6, interpSpeedIncreasing=20}
    boneToModify {boneName=dyn_fx_chest_trail_fr_l_MSTR}
   }
   n106 = "Transform (Modify) Bone - Bone: dyn_fx_chest_trail_fr_r_MSTR"!(ComponentPose=@n105, Alpha=0.6, Rotation=Get sineRotation3() -> (sineRotation3:Rotator)) -> (Pose) {
    alpha=0
    rotationMode=BMM_Additive
    rotationSpace=BCS_BoneSpace
    translationMode=BMM_Additive
    translationSpace=BCS_BoneSpace
    alphaBoolBlend {blendOption=Linear}
    alphaScaleBiasClamp {bInterpResult=true, clampMax=0.6, interpSpeedIncreasing=20}
    boneToModify {boneName=dyn_fx_chest_trail_fr_r_MSTR}
   }
   n107 = "Transform (Modify) Bone - Bone: dyn_fx_waist_trail_fr_l_MSTR"!(ComponentPose=@n106, Alpha=0.4, Rotation=Get sineRotation4() -> (sineRotation4:Rotator)) -> (Pose) {
    alpha=0
    rotationMode=BMM_Additive
    rotationSpace=BCS_BoneSpace
    translationMode=BMM_Additive
    translationSpace=BCS_BoneSpace
    alphaBoolBlend {blendOption=Linear}
    alphaScaleBiasClamp {bInterpResult=true, clampMax=0.6, interpSpeedIncreasing=20}
    boneToModify {boneName=dyn_fx_waist_trail_fr_l_MSTR}
   }
   n108 = "Transform (Modify) Bone - Bone: dyn_fx_waist_trail_fr_r_MSTR"!(ComponentPose=@n107, Alpha=0.4, Rotation=Get sineRotation2() -> (sineRotation2:Rotator)) -> (Pose) {
    alpha=0
    rotationMode=BMM_Additive
    rotationSpace=BCS_BoneSpace
    translationMode=BMM_Additive
    translationSpace=BCS_BoneSpace
    alphaBoolBlend {blendOption=Linear}
    alphaScaleBiasClamp {bInterpResult=true, clampMax=0.6, interpSpeedIncreasing=20}
    boneToModify {boneName=dyn_fx_waist_trail_fr_r_MSTR}
   }
   n109 = Get TargetingAlpha() -> (TargetingAlpha:float)
   n110 = Anim Dynamics - Bone: dyn_fx_chest_trail_bk_l_01(ComponentPose=@n108, Alpha=@n109, ExternalForce="0,0,0") {
    angularDampingOverride=0.9999
    angularSpringConstant=500
    bAngularSpring=true
    bGravityOverrideInSimSpace=true
    bOverrideAngularDamping=true
    bUseGravityOverride=true
    gravityScale=-0.5
    simulationSpace=RootRelative
    alphaBoolBlend {blendOption=Linear}
    boundBone {boneName=dyn_fx_chest_trail_bk_l_01}
    componentAppliedLinearAccClamp=<3500 3500 3500>
    componentLinearAccScale=<0.8 0.8 0.5>
    componentLinearVelScale=<0.8 0.8 0.5>
    gravityOverride {z=100}
    physicsBodyDefinitions [{
      sphereCollisionRadius=0
      boundBone {boneName=dyn_fx_chest_trail_bk_l_01}
      boxExtents=<10 25 25>
      constraintSetup {angularLimitsMax=<-1 45 40>, angularLimitsMin=<-1 -45 -10>, angularTarget=<0 0 0>}
      localJointOffset=<4 0 0>
     }]
    relativeSpaceBone {boneName=pelvis}
   }
   n111 = Anim Dynamics - Bone: dyn_fx_chest_trail_bk_l_02(ComponentPose=@n110, Alpha=@n109, ExternalForce="0,0,0") {
    angularDampingOverride=0.9999
    angularSpringConstant=500
    bAngularSpring=true
    bGravityOverrideInSimSpace=true
    bOverrideAngularDamping=true
    bUseGravityOverride=true
    gravityScale=-0.5
    simulationSpace=RootRelative
    alphaBoolBlend {blendOption=Linear}
    boundBone {boneName=dyn_fx_chest_trail_bk_l_02}
    componentAppliedLinearAccClamp=<3500 3500 3500>
    componentLinearAccScale=<0.8 0.8 0.5>
    componentLinearVelScale=<0.8 0.8 0.5>
    gravityOverride {z=100}
    physicsBodyDefinitions [{
      sphereCollisionRadius=0
      boundBone {boneName=dyn_fx_chest_trail_bk_l_02}
      boxExtents=<10 25 25>
      constraintSetup {angularLimitsMax=<5 45 20>, angularLimitsMin=<-5 -45 -10>, angularTarget=<0 0 0>}
      localJointOffset=<4 0 0>
     }]
    relativeSpaceBone {boneName=pelvis}
   }
   n112 = Anim Dynamics - Bone: dyn_fx_chest_trail_bk_l_03(ComponentPose=@n111, Alpha=@n109, ExternalForce="0,0,0") {
    angularDampingOverride=0.9999
    angularSpringConstant=500
    bAngularSpring=true
    bGravityOverrideInSimSpace=true
    bOverrideAngularDamping=true
    bUseGravityOverride=true
    gravityScale=-0.5
    simulationSpace=RootRelative
    alphaBoolBlend {blendOption=Linear}
    boundBone {boneName=dyn_fx_chest_trail_bk_l_03}
    componentAppliedLinearAccClamp=<3500 3500 3500>
    componentLinearAccScale=<0.8 0.8 0.5>
    componentLinearVelScale=<0.8 0.8 0.5>
    gravityOverride {z=100}
    physicsBodyDefinitions [{
      sphereCollisionRadius=0
      boundBone {boneName=dyn_fx_chest_trail_bk_l_03}
      boxExtents=<10 25 25>
      constraintSetup {angularLimitsMax=<5 45 20>, angularLimitsMin=<-5 -45 -10>, angularTarget=<0 0 0>}
      localJointOffset=<4 0 0>
     }]
    relativeSpaceBone {boneName=pelvis}
   }
   n113 = Anim Dynamics - Bone: dyn_fx_chest_trail_bk_r_01(ComponentPose=@n112, Alpha=@n109, ExternalForce="0,0,0") {
    angularDampingOverride=0.9999
    angularSpringConstant=500
    bAngularSpring=true
    bGravityOverrideInSimSpace=true
    bOverrideAngularDamping=true
    bUseGravityOverride=true
    gravityScale=-0.5
    simulationSpace=RootRelative
    alphaBoolBlend {blendOption=Linear}
    boundBone {boneName=dyn_fx_chest_trail_bk_r_01}
    componentAppliedLinearAccClamp=<3500 3500 3500>
    componentLinearAccScale=<0.8 0.8 0.5>
    componentLinearVelScale=<0.8 0.8 0.5>
    gravityOverride {z=100}
    physicsBodyDefinitions [{
      sphereCollisionRadius=0
      boundBone {boneName=dyn_fx_chest_trail_bk_r_01}
      boxExtents=<10 25 25>
      constraintSetup {angularLimitsMax=<-1 45 15>, angularLimitsMin=<-1 -45 -45>, angularTarget=<0 0 0>}
      localJointOffset=<4 0 0>
     }]
    relativeSpaceBone {boneName=pelvis}
   }
   n114 = Anim Dynamics - Bone: dyn_fx_chest_trail_bk_r_02(ComponentPose=@n113, Alpha=@n109, ExternalForce="0,0,0") {
    angularDampingOverride=0.9999
    angularSpringConstant=500
    bAngularSpring=true
    bGravityOverrideInSimSpace=true
    bOverrideAngularDamping=true
    bUseGravityOverride=true
    gravityScale=-0.5
    simulationSpace=RootRelative
    alphaBoolBlend {blendOption=Linear}
    boundBone {boneName=dyn_fx_chest_trail_bk_r_02}
    componentAppliedLinearAccClamp=<3500 3500 3500>
    componentLinearAccScale=<0.8 0.8 0.5>
    componentLinearVelScale=<0.8 0.8 0.5>
    gravityOverride {z=100}
    physicsBodyDefinitions [{
      sphereCollisionRadius=0
      boundBone {boneName=dyn_fx_chest_trail_bk_r_02}
      boxExtents=<10 25 25>
      constraintSetup {angularLimitsMax=<5 45 15>, angularLimitsMin=<-5 -45 -25>, angularTarget=<0 0 0>}
      localJointOffset=<4 0 0>
     }]
    relativeSpaceBone {boneName=pelvis}
   }
   n115 = Anim Dynamics - Bone: dyn_fx_chest_trail_bk_r_03(ComponentPose=@n114, Alpha=@n109, ExternalForce="0,0,0") {
    angularDampingOverride=0.9999
    angularSpringConstant=500
    bAngularSpring=true
    bGravityOverrideInSimSpace=true
    bOverrideAngularDamping=true
    bUseGravityOverride=true
    gravityScale=-0.5
    simulationSpace=RootRelative
    alphaBoolBlend {blendOption=Linear}
    boundBone {boneName=dyn_fx_chest_trail_bk_r_03}
    componentAppliedLinearAccClamp=<3500 3500 3500>
    componentLinearAccScale=<0.8 0.8 0.5>
    componentLinearVelScale=<0.8 0.8 0.5>
    gravityOverride {z=100}
    physicsBodyDefinitions [{
      sphereCollisionRadius=0
      boundBone {boneName=dyn_fx_chest_trail_bk_r_03}
      boxExtents=<10 25 25>
      constraintSetup {angularLimitsMax=<5 45 15>, angularLimitsMin=<-5 -45 -25>, angularTarget=<0 0 0>}
      localJointOffset=<4 0 0>
     }]
    relativeSpaceBone {boneName=pelvis}
   }
   n116 = Get TargetingAlpha()
   n117 = Anim Dynamics - Bone: dyn_fx_chest_trail_fr_l_01(ComponentPose=@n115, Alpha=@n116, ExternalForce="0,0,0") {
    angularDampingOverride=0.9999
    angularSpringConstant=500
    bAngularSpring=true
    bGravityOverrideInSimSpace=true
    bOverrideAngularDamping=true
    bUseGravityOverride=true
    gravityScale=-1
    simulationSpace=BoneRelative
    alphaBoolBlend {blendOption=Linear}
    boundBone {boneName=dyn_fx_chest_trail_fr_l_01}
    componentAppliedLinearAccClamp=<3500 3500 3500>
    componentLinearAccScale=<1.2 1.2 0.8>
    componentLinearVelScale=<1.2 1.2 0.8>
    gravityOverride {z=100}
    physicsBodyDefinitions [{
      sphereCollisionRadius=0
      boundBone {boneName=dyn_fx_chest_trail_fr_l_01}
      boxExtents=<15 20 20>
      constraintSetup {angularLimitsMax=<1 2 10>, angularLimitsMin=<-1 -25 -10>, angularTarget=<0 0 0>}
      localJointOffset=<3 0 0>
     }]
    relativeSpaceBone {boneName=spine_01}
   }
   n118 = Anim Dynamics - Bone: dyn_fx_chest_trail_fr_l_02(ComponentPose=@n117, Alpha=@n116, ExternalForce="0,0,0") {
    angularDampingOverride=0.9999
    angularSpringConstant=500
    bAngularSpring=true
    bGravityOverrideInSimSpace=true
    bOverrideAngularDamping=true
    bUseGravityOverride=true
    gravityScale=-1
    simulationSpace=BoneRelative
    alphaBoolBlend {blendOption=Linear}
    boundBone {boneName=dyn_fx_chest_trail_fr_l_02}
    componentAppliedLinearAccClamp=<3500 3500 3500>
    componentLinearAccScale=<1.2 1.2 0.8>
    componentLinearVelScale=<1.2 1.2 0.8>
    gravityOverride {z=100}
    physicsBodyDefinitions [{
      sphereCollisionRadius=0
      boundBone {boneName=dyn_fx_chest_trail_fr_l_02}
      boxExtents=<15 20 20>
      constraintSetup {angularLimitsMax=<1 2 10>, angularLimitsMin=<-1 -25 -25>, angularTarget=<0 0 0>}
      localJointOffset=<3 0 0>
     }]
    relativeSpaceBone {boneName=spine_01}
   }
   n119 = Anim Dynamics - Bone: dyn_fx_chest_trail_fr_l_03(ComponentPose=@n118, Alpha=@n116, ExternalForce="0,0,0") {
    angularDampingOverride=0.9999
    angularSpringConstant=500
    bAngularSpring=true
    bGravityOverrideInSimSpace=true
    bOverrideAngularDamping=true
    bUseGravityOverride=true
    gravityScale=-1
    simulationSpace=BoneRelative
    alphaBoolBlend {blendOption=Linear}
    boundBone {boneName=dyn_fx_chest_trail_fr_l_03}
    componentAppliedLinearAccClamp=<3500 3500 3500>
    componentLinearAccScale=<1.2 1.2 0.8>
    componentLinearVelScale=<1.2 1.2 0.8>
    gravityOverride {z=100}
    physicsBodyDefinitions [{
      sphereCollisionRadius=0
      boundBone {boneName=dyn_fx_chest_trail_fr_l_03}
      boxExtents=<15 20 20>
      constraintSetup {angularLimitsMax=<1 2 10>, angularLimitsMin=<-1 -25 -25>, angularTarget=<0 0 0>}
      localJointOffset=<3 0 0>
     }]
    relativeSpaceBone {boneName=spine_01}
   }
   n120 = Anim Dynamics - Bone: dyn_fx_chest_trail_fr_r_01(ComponentPose=@n119, Alpha=@n116, ExternalForce="0,0,0") {
    angularDampingOverride=0.9999
    angularSpringConstant=500
    bAngularSpring=true
    bGravityOverrideInSimSpace=true
    bOverrideAngularDamping=true
    bUseGravityOverride=true
    gravityScale=-1
    simulationSpace=BoneRelative
    alphaBoolBlend {blendOption=Linear}
    boundBone {boneName=dyn_fx_chest_trail_fr_r_01}
    componentAppliedLinearAccClamp=<3500 3500 3500>
    componentLinearAccScale=<1.2 1.2 0.8>
    componentLinearVelScale=<1.2 1.2 0.8>
    gravityOverride {z=100}
    physicsBodyDefinitions [{
      sphereCollisionRadius=0
      boundBone {boneName=dyn_fx_chest_trail_fr_r_01}
      boxExtents=<15 20 20>
      constraintSetup {angularLimitsMax=<1 2 10>, angularLimitsMin=<-1 -25 -10>, angularTarget=<0 0 0>}
      localJointOffset=<3 0 0>
     }]
    relativeSpaceBone {boneName=spine_01}
   }
   n121 = Anim Dynamics - Bone: dyn_fx_chest_trail_fr_r_02(ComponentPose=@n120, Alpha=@n116, ExternalForce="0,0,0") {
    angularDampingOverride=0.9999
    angularSpringConstant=500
    bAngularSpring=true
    bGravityOverrideInSimSpace=true
    bOverrideAngularDamping=true
    bUseGravityOverride=true
    gravityScale=-1
    simulationSpace=BoneRelative
    alphaBoolBlend {blendOption=Linear}
    boundBone {boneName=dyn_fx_chest_trail_fr_r_02}
    componentAppliedLinearAccClamp=<3500 3500 3500>
    componentLinearAccScale=<1.2 1.2 0.8>
    componentLinearVelScale=<1.2 1.2 0.8>
    gravityOverride {z=100}
    physicsBodyDefinitions [{
      sphereCollisionRadius=0
      boundBone {boneName=dyn_fx_chest_trail_fr_r_02}
      boxExtents=<15 20 20>
      constraintSetup {angularLimitsMax=<1 2 20>, angularLimitsMin=<-1 -25 -15>, angularTarget=<0 0 0>}
      localJointOffset=<3 0 0>
     }]
    relativeSpaceBone {boneName=spine_01}
   }
   n122 = Anim Dynamics - Bone: dyn_fx_chest_trail_fr_r_03(ComponentPose=@n121, Alpha=@n116, ExternalForce="0,0,0") {
    angularDampingOverride=0.9999
    angularSpringConstant=500
    bAngularSpring=true
    bGravityOverrideInSimSpace=true
    bOverrideAngularDamping=true
    bUseGravityOverride=true
    gravityScale=-1
    simulationSpace=BoneRelative
    alphaBoolBlend {blendOption=Linear}
    boundBone {boneName=dyn_fx_chest_trail_fr_r_03}
    componentAppliedLinearAccClamp=<3500 3500 3500>
    componentLinearAccScale=<1.2 1.2 0.8>
    componentLinearVelScale=<1.2 1.2 0.8>
    gravityOverride {z=100}
    physicsBodyDefinitions [{
      sphereCollisionRadius=0
      boundBone {boneName=dyn_fx_chest_trail_fr_r_03}
      boxExtents=<15 20 20>
      constraintSetup {angularLimitsMax=<1 2 25>, angularLimitsMin=<-1 -25 -10>, angularTarget=<0 0 0>}
      localJointOffset=<3 0 0>
     }]
    relativeSpaceBone {boneName=spine_01}
   }
   n123 = Get TargetingAlpha()
   n124 = Anim Dynamics - Bone: dyn_fx_waist_trail_fr_l_01(ComponentPose=@n122, Alpha=@n123, ExternalForce="0,0,0") {
    angularDampingOverride=0.9999
    angularSpringConstant=500
    bAngularSpring=true
    bGravityOverrideInSimSpace=true
    bOverrideAngularDamping=true
    bUseGravityOverride=true
    gravityScale=-1
    simulationSpace=BoneRelative
    alphaBoolBlend {blendOption=Linear}
    boundBone {boneName=dyn_fx_waist_trail_fr_l_01}
    componentAppliedLinearAccClamp=<3500 3500 3500>
    componentLinearAccScale=<1.2 1.2 0.8>
    componentLinearVelScale=<1.2 1.2 0.8>
    gravityOverride {z=100}
    physicsBodyDefinitions [{
      sphereCollisionRadius=0
      boundBone {boneName=dyn_fx_waist_trail_fr_l_01}
      boxExtents=<15 20 20>
      constraintSetup {angularLimitsMax=<1 8 15>, angularLimitsMin=<-1 -45 -10>, angularTarget=<0 0 0>}
      localJointOffset=<3 0 0>
     }]
    relativeSpaceBone {boneName=pelvis}
   }
   n125 = Anim Dynamics - Bone: dyn_fx_waist_trail_fr_l_02(ComponentPose=@n124, Alpha=@n123, ExternalForce="0,0,0") {
    angularDampingOverride=0.9999
    angularSpringConstant=500
    bAngularSpring=true
    bGravityOverrideInSimSpace=true
    bOverrideAngularDamping=true
    bUseGravityOverride=true
    gravityScale=-1
    simulationSpace=BoneRelative
    alphaBoolBlend {blendOption=Linear}
    boundBone {boneName=dyn_fx_waist_trail_fr_l_02}
    componentAppliedLinearAccClamp=<3500 3500 3500>
    componentLinearAccScale=<1.2 1.2 0.8>
    componentLinearVelScale=<1.2 1.2 0.8>
    gravityOverride {z=100}
    physicsBodyDefinitions [{
      sphereCollisionRadius=0
      boundBone {boneName=dyn_fx_waist_trail_fr_l_02}
      boxExtents=<15 20 20>
      constraintSetup {angularLimitsMax=<1 8 20>, angularLimitsMin=<-1 -45 -15>, angularTarget=<0 0 0>}
      localJointOffset=<3 0 0>
     }]
    relativeSpaceBone {boneName=pelvis}
   }
   n126 = Anim Dynamics - Bone: dyn_fx_waist_trail_fr_l_03(ComponentPose=@n125, Alpha=1, ExternalForce="0,0,0") {
    angularDampingOverride=0.9999
    angularSpringConstant=500
    bAngularSpring=true
    bGravityOverrideInSimSpace=true
    bOverrideAngularDamping=true
    bUseGravityOverride=true
    gravityScale=-1
    simulationSpace=BoneRelative
    alphaBoolBlend {blendOption=Linear}
    boundBone {boneName=dyn_fx_waist_trail_fr_l_03}
    componentAppliedLinearAccClamp=<3500 3500 3500>
    componentLinearAccScale=<1.2 1.2 0.8>
    componentLinearVelScale=<1.2 1.2 0.8>
    gravityOverride {z=100}
    physicsBodyDefinitions [{
      sphereCollisionRadius=0
      boundBone {boneName=dyn_fx_waist_trail_fr_l_03}
      boxExtents=<15 20 20>
      constraintSetup {angularLimitsMax=<1 8 20>, angularLimitsMin=<-1 -45 -15>, angularTarget=<0 0 0>}
      localJointOffset=<3 0 0>
     }]
    relativeSpaceBone {boneName=pelvis}
   }
   n127 = Anim Dynamics - Bone: dyn_fx_waist_trail_fr_r_01(ComponentPose=@n126, Alpha=@n123, ExternalForce="0,0,0") {
    angularDampingOverride=0.9999
    angularSpringConstant=500
    bAngularSpring=true
    bGravityOverrideInSimSpace=true
    bOverrideAngularDamping=true
    bUseGravityOverride=true
    gravityScale=-1
    simulationSpace=BoneRelative
    alphaBoolBlend {blendOption=Linear}
    boundBone {boneName=dyn_fx_waist_trail_fr_r_01}
    componentAppliedLinearAccClamp=<3500 3500 3500>
    componentLinearAccScale=<1.2 1.2 0.8>
    componentLinearVelScale=<1.2 1.2 0.8>
    gravityOverride {z=100}
    physicsBodyDefinitions [{
      sphereCollisionRadius=0
      boundBone {boneName=dyn_fx_waist_trail_fr_r_01}
      boxExtents=<15 20 20>
      constraintSetup {angularLimitsMax=<1 8 15>, angularLimitsMin=<-1 -45 -10>, angularTarget=<0 0 0>}
      localJointOffset=<3 0 0>
     }]
    relativeSpaceBone {boneName=pelvis}
   }
   n128 = Anim Dynamics - Bone: dyn_fx_waist_trail_fr_r_02(ComponentPose=@n127, Alpha=@n123, ExternalForce="0,0,0") {
    angularDampingOverride=0.9999
    angularSpringConstant=500
    bAngularSpring=true
    bGravityOverrideInSimSpace=true
    bOverrideAngularDamping=true
    bUseGravityOverride=true
    gravityScale=-1
    simulationSpace=BoneRelative
    alphaBoolBlend {blendOption=Linear}
    boundBone {boneName=dyn_fx_waist_trail_fr_r_02}
    componentAppliedLinearAccClamp=<3500 3500 3500>
    componentLinearAccScale=<1.2 1.2 0.8>
    componentLinearVelScale=<1.2 1.2 0.8>
    gravityOverride {z=100}
    physicsBodyDefinitions [{
      sphereCollisionRadius=0
      boundBone {boneName=dyn_fx_waist_trail_fr_r_02}
      boxExtents=<15 20 20>
      constraintSetup {angularLimitsMax=<1 8 20>, angularLimitsMin=<-1 -45 -25>, angularTarget=<0 0 0>}
      localJointOffset=<3 0 0>
     }]
    relativeSpaceBone {boneName=pelvis}
   }
   n129 = Anim Dynamics - Bone: dyn_fx_waist_trail_fr_r_03(ComponentPose=@n128, Alpha=1, ExternalForce="0,0,0") {
    angularDampingOverride=0.9999
    angularSpringConstant=500
    bAngularSpring=true
    bGravityOverrideInSimSpace=true
    bOverrideAngularDamping=true
    bUseGravityOverride=true
    gravityScale=-1
    simulationSpace=BoneRelative
    alphaBoolBlend {blendOption=Linear}
    boundBone {boneName=dyn_fx_waist_trail_fr_r_03}
    componentAppliedLinearAccClamp=<3500 3500 3500>
    componentLinearAccScale=<1.2 1.2 0.8>
    componentLinearVelScale=<1.2 1.2 0.8>
    gravityOverride {z=100}
    physicsBodyDefinitions [{
      sphereCollisionRadius=0
      boundBone {boneName=dyn_fx_waist_trail_fr_r_03}
      boxExtents=<15 20 20>
      constraintSetup {angularLimitsMax=<1 8 20>, angularLimitsMin=<-1 -45 -25>, angularTarget=<0 0 0>}
      localJointOffset=<3 0 0>
     }]
    relativeSpaceBone {boneName=pelvis}
   }
   Output Pose(Result=Component To Local(ComponentPose=@n129)) {layerGroup=None}
   n130 = "Constraint - dyn_fx_waist_trail_fr_l_MSTR (PPPPPPPPP)"!(
    ComponentPose=Local To Component(LocalPose=Copy Pose From Mesh() {bUseAttachedParent=true})
    Alpha=1
    ConstraintWeights_0=0.1727
    ConstraintWeights_1:float=0.1117
    ConstraintWeights_2:float=0.015
    ConstraintWeights_3:float=0.2395
    ConstraintWeights_4:float=0.0879
    ConstraintWeights_5:float=0.0505
    ConstraintWeights_6:float=0.0308
    ConstraintWeights_7:float=0.0298
    ConstraintWeights_8:float=0.2622
   ) -> (Pose) {
    alphaBoolBlend {blendOption=Linear}
    boneToModify {boneName=dyn_fx_waist_trail_fr_l_MSTR}
    constraintSetup [
     {targetBone {boneName=dyn_coat_fr_l_03}}
     {targetBone {boneName=dyn_coat_l_02}}
     {targetBone {boneName=dyn_coat_fr_l_01}}
     {targetBone {boneName=dyn_coat_fr_l_02}}
     {targetBone {boneName=dyn_coat_l_01}}
     {targetBone {boneName=dyn_coat_fr_l_04}}
     {targetBone {boneName=dyn_coat_bk_l_02}}
     {targetBone {boneName=dyn_coat_l_04}}
     {targetBone {boneName=dyn_coat_l_03}}
    ]
    constraintWeights [1,1,1,1,1,1,1,1,1]
   }
   n131 = "Constraint - dyn_fx_waist_trail_bk_MSTR (PPP)"!(ComponentPose=@n130, Alpha=1, ConstraintWeights_0=0.9634, ConstraintWeights_1:float=0.0232, ConstraintWeights_2:float=0.0134) -> (Pose) {
    alphaBoolBlend {blendOption=Linear}
    boneToModify {boneName=dyn_fx_waist_trail_bk_MSTR}
    constraintSetup [{targetBone {boneName=dyn_coat_bk_02}}, {targetBone {boneName=dyn_coat_bk_01}}, {targetBone {boneName=dyn_coat_bk_r_02}}]
    constraintWeights [1,1,1]
   }
   n132 = "Constraint - dyn_fx_waist_trail_fr_r_MSTR (PPPPPPPPP)"!(
    ComponentPose=@n131
    Alpha=1
    ConstraintWeights_0=0.0431
    ConstraintWeights_1:float=0.2267
    ConstraintWeights_2:float=0.1831
    ConstraintWeights_3:float=0.0104
    ConstraintWeights_4:float=0.0752
    ConstraintWeights_5:float=0.1185
    ConstraintWeights_6:float=0.2982
    ConstraintWeights_7:float=0.0221
    ConstraintWeights_8:float=0.0228
   ) -> (Pose) {
    alphaBoolBlend {blendOption=Linear}
    boneToModify {boneName=dyn_fx_waist_trail_fr_r_MSTR}
    constraintSetup [
     {targetBone {boneName=dyn_coat_fr_r_04}}
     {targetBone {boneName=dyn_coat_fr_r_02}}
     {targetBone {boneName=dyn_coat_fr_r_03}}
     {targetBone {boneName=dyn_coat_fr_r_01}}
     {targetBone {boneName=dyn_coat_r_01}}
     {targetBone {boneName=dyn_coat_r_02}}
     {targetBone {boneName=dyn_coat_r_03}}
     {targetBone {boneName=dyn_coat_r_04}}
     {targetBone {boneName=dyn_coat_bk_r_02}}
    ]
    constraintWeights [1,1,1,1,1,1,1,1,1]
   }
   n133 = "Constraint - dyn_fx_chest_trail_fr_r_MSTR (PPPPPP)"!(
    ComponentPose=@n132
    Alpha=1
    ConstraintWeights_0=0.0114
    ConstraintWeights_1:float=0.1035
    ConstraintWeights_2:float=0.2374
    ConstraintWeights_3:float=0.4801
    ConstraintWeights_4:float=0.1404
    ConstraintWeights_5:float=0.0272
   ) -> (Pose) {
    alphaBoolBlend {blendOption=Linear}
    boneToModify {boneName=dyn_fx_chest_trail_fr_r_MSTR}
    constraintSetup [
     {targetBone {boneName=spine_01}}
     {targetBone {boneName=spine_02}}
     {targetBone {boneName=spine_03}}
     {targetBone {boneName=spine_04}}
     {targetBone {boneName=spine_05}}
     {targetBone {boneName=clavicle_r}}
    ]
    constraintWeights [1,1,1,1,1,1]
   }
   n134 = "Constraint - dyn_fx_chest_trail_bk_r_MSTR (PPPPPP)"!(
    ComponentPose=@n133
    Alpha=1
    ConstraintWeights_0=0.0426
    ConstraintWeights_1:float=0.0273
    ConstraintWeights_2:float=0.3304
    ConstraintWeights_3:float=0.2252
    ConstraintWeights_4:float=0.1294
    ConstraintWeights_5:float=0.245
   ) -> (Pose) {
    alphaBoolBlend {blendOption=Linear}
    boneToModify {boneName=dyn_fx_chest_trail_bk_r_MSTR}
    constraintSetup [
     {targetBone {boneName=upperarm_twist_01_r}}
     {targetBone {boneName=spine_02}}
     {targetBone {boneName=spine_03}}
     {targetBone {boneName=spine_04}}
     {targetBone {boneName=spine_05}}
     {targetBone {boneName=clavicle_r}}
    ]
    constraintWeights [1,1,1,1,1,1]
   }
   n135 = "Constraint - dyn_fx_chest_trail_fr_l_MSTR (PPPPPP)"!(
    ComponentPose=@n134
    Alpha=1
    ConstraintWeights_0=0.0272
    ConstraintWeights_1:float=0.0114
    ConstraintWeights_2:float=0.1035
    ConstraintWeights_3:float=0.2374
    ConstraintWeights_4:float=0.4801
    ConstraintWeights_5:float=0.1404
   ) -> (Pose) {
    alphaBoolBlend {blendOption=Linear}
    boneToModify {boneName=dyn_fx_chest_trail_fr_l_MSTR}
    constraintSetup [
     {targetBone {boneName=clavicle_l}}
     {targetBone {boneName=spine_01}}
     {targetBone {boneName=spine_02}}
     {targetBone {boneName=spine_03}}
     {targetBone {boneName=spine_04}}
     {targetBone {boneName=spine_05}}
    ]
    constraintWeights [1,1,1,1,1,1]
   }
   n136 = "Constraint - dyn_fx_chest_trail_bk_l_MSTR (PPPPP)"!(
    ComponentPose=@n135
    Alpha=1
    ConstraintWeights_0=0.0403
    ConstraintWeights_1:float=0.3374
    ConstraintWeights_2:float=0.2564
    ConstraintWeights_3:float=0.2165
    ConstraintWeights_4:float=0.1493
   ) -> (Pose) {
    alphaBoolBlend {blendOption=Linear}
    boneToModify {boneName=dyn_fx_chest_trail_bk_l_MSTR}
    constraintSetup [
     {targetBone {boneName=spine_02}}
     {targetBone {boneName=spine_03}}
     {targetBone {boneName=spine_04}}
     {targetBone {boneName=spine_05}}
     {targetBone {boneName=clavicle_l}}
    ]
    constraintWeights [1,1,1,1,1]
   }
   n137 = Get InVehicleScaleAlpha() -> (InVehicleScaleAlpha:float)
   n138 = "Transform (Modify) Bone - Bone: dyn_coat_l_02"(ComponentPose=@n136, Alpha=@n137, Translation="0,0,0", Rotation="0,0,0", Scale="0.1,0.1,0.1") {scaleMode=BMM_Replace, alphaBoolBlend {blendOption=Linear}, boneToModify {boneName=dyn_coat_l_02}, scale=<0.1 0.1 0.1>}
   n139 = "Transform (Modify) Bone - Bone: dyn_coat_fr_l_02"!(ComponentPose=@n138, Alpha=@n137, Translation="0,0,0", Rotation="0,0,0", Scale:Vector="0.1,0.1,0.1") -> (Pose) {scaleMode=BMM_Replace, alphaBoolBlend {blendOption=Linear}, boneToModify {boneName=dyn_coat_fr_l_02}, scale=<0.1 0.1 0.1>}
   n140 = "Transform (Modify) Bone - Bone: dyn_coat_bk_02"(ComponentPose=@n139, Alpha=@n137, Translation="0,0,0", Rotation="0,0,0", Scale="0.1,0.1,0.1") {scaleMode=BMM_Replace, alphaBoolBlend {blendOption=Linear}, boneToModify {boneName=dyn_coat_bk_02}, scale=<0.1 0.1 0.1>}
   n141 = "Transform (Modify) Bone - Bone: dyn_coat_bk_l_02"(ComponentPose=@n140, Alpha=@n137, Translation="0,0,0", Rotation="0,0,0", Scale="0.1,0.1,0.1") {scaleMode=BMM_Replace, alphaBoolBlend {blendOption=Linear}, boneToModify {boneName=dyn_coat_bk_l_02}, scale=<0.1 0.1 0.1>}
   n142 = "Transform (Modify) Bone - Bone: dyn_coat_bk_r_02"(ComponentPose=@n141, Alpha=@n137, Translation="0,0,0", Rotation="0,0,0", Scale="0.1,0.1,0.1") {scaleMode=BMM_Replace, alphaBoolBlend {blendOption=Linear}, boneToModify {boneName=dyn_coat_bk_r_02}, scale=<0.1 0.1 0.1>}
   n143 = "Transform (Modify) Bone - Bone: dyn_coat_fr_r_02"(ComponentPose=@n142, Alpha=@n137, Translation="0,0,0", Rotation="0,0,0", Scale="0.1,0.1,0.1") {scaleMode=BMM_Replace, alphaBoolBlend {blendOption=Linear}, boneToModify {boneName=dyn_coat_fr_r_02}, scale=<0.1 0.1 0.1>}
   n144 = "Transform (Modify) Bone - Bone: dyn_coat_r_02"(ComponentPose=@n143, Alpha=@n137, Translation="0,0,0", Rotation="0,0,0", Scale="0.1,0.1,0.1") {scaleMode=BMM_Replace, alphaBoolBlend {blendOption=Linear}, boneToModify {boneName=dyn_coat_r_02}, scale=<0.1 0.1 0.1>}
   n145 = "Constraint - dyn_zipper_chest_MSTR (PPPPP)"!(
    ComponentPose=@n144
    Alpha=1
    ConstraintWeights_0=0.0613
    ConstraintWeights_1:float=0.136
    ConstraintWeights_2:float=0.0294
    ConstraintWeights_3:float=0.0255
    ConstraintWeights_4:float=0.7477
   ) -> (Pose) {
    alphaBoolBlend {blendOption=Linear}
    boneToModify {boneName=dyn_zipper_chest_MSTR}
    constraintSetup [
     {targetBone {boneName=neck_02}}
     {targetBone {boneName=neck_01}}
     {targetBone {boneName=spine_04}}
     {targetBone {boneName=head}}
     {targetBone {boneName=spine_05}}
    ]
    constraintWeights [1,1,1,1,1]
   }
   n146 = "Transform (Modify) Bone - Bone: dyn_coat_bk_02"(ComponentPose=@n145, Alpha=1, Translation="0,0,0.6", Rotation="0,0,0", Scale="1,1,1") {
    scaleMode=BMM_Replace
    translationMode=BMM_Replace
    translationSpace=BCS_BoneSpace
    alphaBoolBlend {blendOption=Linear}
    boneToModify {boneName=dyn_coat_bk_02}
    translation {z=0.6}
   }
   n147 = Get ThighRotLeft() -> (ThighRotLeft:float)
   n148 = "Transform (Modify) Bone - Bone: thigh_volume_bk_l"(ComponentPose=@n146, Alpha=@n147, Translation="4.485,-5.330,0.409", Rotation="0,0,0", Scale="1,1,1") {
    rotationSpace=BCS_BoneSpace
    scaleMode=BMM_Additive
    scaleSpace=BCS_BoneSpace
    translationMode=BMM_Additive
    translationSpace=BCS_BoneSpace
    alphaBoolBlend {blendOption=Linear}
    boneToModify {boneName=thigh_volume_bk_l}
    translation=<4.485 -5.3303 0.409>
   }
   n149 = "Transform (Modify) Bone - Bone: thigh_volume_fr_l"(ComponentPose=@n148, Alpha=@n147, Translation="-2.144,7.226,-0.000", Rotation="0,0,0", Scale="1,1,1") {
    rotationSpace=BCS_BoneSpace
    scaleMode=BMM_Additive
    scaleSpace=BCS_BoneSpace
    translationMode=BMM_Additive
    translationSpace=BCS_BoneSpace
    alphaBoolBlend {blendOption=Linear}
    boneToModify {boneName=thigh_volume_fr_l}
    translation=<-2.144 7.226>
   }
   n150 = Get ThighRotRight() -> (ThighRotRight:float)
   n151 = "Transform (Modify) Bone - Bone: thigh_volume_bk_r"(ComponentPose=@n149, Alpha=@n150, Translation="4.485,5.33,0.000", Rotation="0,0,0", Scale="1,1,1") {
    rotationSpace=BCS_BoneSpace
    scaleMode=BMM_Additive
    scaleSpace=BCS_BoneSpace
    translationMode=BMM_Additive
    translationSpace=BCS_BoneSpace
    alphaBoolBlend {blendOption=Linear}
    boneToModify {boneName=thigh_volume_bk_r}
    translation=<4.485 5.33>
   }
   n152 = "Transform (Modify) Bone - Bone: thigh_volume_fr_r"(ComponentPose=@n151, Alpha=@n150, Translation="-2.144,-7.226,-0.000", Rotation="0,0,0", Scale="1,1,1") {
    rotationSpace=BCS_BoneSpace
    scaleMode=BMM_Additive
    scaleSpace=BCS_BoneSpace
    translationMode=BMM_Additive
    translationSpace=BCS_BoneSpace
    alphaBoolBlend {blendOption=Linear}
    boneToModify {boneName=thigh_volume_fr_r}
    translation=<-2.144 -7.226>
   }
   Save cached pose 'initial Pose'(Pose=Component To Local(ComponentPose=@n152)) {cacheName=initial Pose}
   Get TargetingAlpha()
  ]}
 EventGraph {nodes [
   Event BlueprintUpdateAnimation()
   > n1 = Sequence!(execute) -> (then_0, then_1, then_2:Exec, then_3:Exec, then_4:Exec)
   n2 = Gravity Override State Function DEPRECATED!(
    execute=[@n1.then_0, @n12.then]
    CharacterPartAnimInstance=Self-Reference()
    GameTimeInSeconds=GetGameTimeInSeconds()
    GravityOverrideParams_Skydive_GoingUp_28_#1_WindFrequency_28_#2:Vector="1.5,7,8"
    GravityOverrideParams_Skydive_GoingUp_28_#1_WindAmplitude_31_#3:Vector="200,-500,400"
    GravityOverrideParams_Skydive_GoingUp_28_#1_WindOffset_32_#4:Vector="0,0,0"
    GravityOverrideParams_Skydive_GoingUp_28_#1_GravityOverride_15_#5:Vector="0,0,980"
    GravityOverrideParams_Skydive_GoingUp_28_#1_JointName_27_#6:Name=root
    GravityOverrideParams_Skydive_FloatingDown_29_#7_WindFrequency_28_#2:Vector="1.5,7,8"
    GravityOverrideParams_Skydive_FloatingDown_29_#7_WindAmplitude_31_#3:Vector="200,-500,400"
    GravityOverrideParams_Skydive_FloatingDown_29_#7_WindOffset_32_#4:Vector="0,-600,400"
    GravityOverrideParams_Skydive_FloatingDown_29_#7_GravityOverride_15_#5:Vector="0,0,980"
    GravityOverrideParams_Skydive_FloatingDown_29_#7_JointName_27_#6:Name=root
    GravityOverrideParams_Skydive_DivingDown_30_#8_WindFrequency_28_#2:Vector="1.5,7,8"
    GravityOverrideParams_Skydive_DivingDown_30_#8_WindAmplitude_31_#3:Vector="50,300,100"
    GravityOverrideParams_Skydive_DivingDown_30_#8_WindOffset_32_#4:Vector="0,0,0"
    GravityOverrideParams_Skydive_DivingDown_30_#8_GravityOverride_15_#5:Vector="0,-980,0"
    GravityOverrideParams_Skydive_DivingDown_30_#8_JointName_27_#6:Name=pelvis
    GravityOverrideParams_Skydive_Parachuting_31_#9_WindFrequency_28_#2:Vector="1.5,7,8"
    GravityOverrideParams_Skydive_Parachuting_31_#9_WindAmplitude_31_#3:Vector="200,-500,400"
    GravityOverrideParams_Skydive_Parachuting_31_#9_WindOffset_32_#4:Vector="0,0,0"
    GravityOverrideParams_Skydive_Parachuting_31_#9_GravityOverride_15_#5:Vector="0,-800,-980"
    GravityOverrideParams_Skydive_Parachuting_31_#9_JointName_27_#6:Name=root
    GravityOverrideParams_OnGround_Standing_32_#10:Gravity Override Setting Struct Structure=(WindFrequency_28_#2=(X=0,Y=0,Z=0),WindAmplitude_31_#3=(X=0,Y=0,Z=0),WindOffset_32_#4=(X=0,Y=0,Z=0),GravityOverride_15_#5=(X=0,Y=0,Z=-980),JointName_27_#6="root")
    GravityOverrideParams_OnGround_Moving_33_#11:Gravity Override Setting Struct Structure=(WindFrequency_28_#2=(X=0,Y=0,Z=0),WindAmplitude_31_#3=(X=0,Y=0,Z=0),WindOffset_32_#4=(X=0,Y=0,Z=0),GravityOverride_15_#5=(X=0,Y=0,Z=-980),JointName_27_#6="root")
    GravityOverrideParams_EmoteOrMelee_34_#12:Gravity Override Setting Struct Structure=(WindFrequency_28_#2=(X=0,Y=0,Z=0),WindAmplitude_31_#3=(X=0,Y=0,Z=0),WindOffset_32_#4=(X=0,Y=0,Z=0),GravityOverride_15_#5=(X=0,Y=0,Z=-980),JointName_27_#6="root")
    GravityOverrideParams_Swimming_35_#13:Gravity Override Setting Struct Structure=(WindFrequency_28_#2=(X=0,Y=0,Z=0),WindAmplitude_31_#3=(X=0,Y=0,Z=0),WindOffset_32_#4=(X=0,Y=0,Z=0),GravityOverride_15_#5=(X=-800,Y=0,Z=0),JointName_27_#6="pelvis")
    GravityOverrideParams_Front_End_36_#14:Gravity Override Setting Struct Structure=(WindFrequency_28_#2=(X=0,Y=0,Z=0),WindAmplitude_31_#3=(X=0,Y=0,Z=0),WindOffset_32_#4=(X=0,Y=0,Z=0),GravityOverride_15_#5=(X=0,Y=0,Z=-980),JointName_27_#6="root")
    GravityOverrideParams_DBNO_40_#15:Gravity Override Setting Struct Structure=(WindFrequency_28_#2=(X=0,Y=0,Z=0),WindAmplitude_31_#3=(X=0,Y=0,Z=0),WindOffset_32_#4=(X=0,Y=0,Z=0),GravityOverride_15_#5=(X=-980,Y=400,Z=0),JointName_27_#6="pelvis")
    GravityOverrideParams_DBNO_Carried_41_#16:Gravity Override Setting Struct Structure=(WindFrequency_28_#2=(X=0,Y=0,Z=0),WindAmplitude_31_#3=(X=0,Y=0,Z=0),WindOffset_32_#4=(X=0,Y=0,Z=0),GravityOverride_15_#5=(X=-980,Y=500,Z=-500),JointName_27_#6="pelvis")
    GravityOverrideParams_OnGround_Falling_45_#17:Gravity Override Setting Struct Structure=(WindFrequency_28_#2=(X=0,Y=0,Z=0),WindAmplitude_31_#3=(X=0,Y=0,Z=0),WindOffset_32_#4=(X=0,Y=0,Z=0),GravityOverride_15_#5=(X=0,Y=0,Z=-980),JointName_27_#6="root")
    GravityOverrideParams_Ziplining_46_#18:Gravity Override Setting Struct Structure=(WindFrequency_28_#2=(X=0,Y=0,Z=0),WindAmplitude_31_#3=(X=0,Y=0,Z=0),WindOffset_32_#4=(X=0,Y=0,Z=0),GravityOverride_15_#5=(X=0,Y=0,Z=-980),JointName_27_#6="root")
    GravityOverrideParams_Skydive_Surfing_49_#19:Gravity Override Setting Struct Structure=(WindFrequency_28_#2=(X=1.5,Y=7,Z=8),WindAmplitude_31_#3=(X=200,Y=-500,Z=400),WindOffset_32_#4=(X=0,Y=0,Z=0),GravityOverride_15_#5=(X=500,Y=-300,Z=-980),JointName_27_#6="root")
    GravityOverrideParams_Riding_Standing_54_#20:Gravity Override Setting Struct Structure=(WindFrequency_28_#2=(X=0,Y=0,Z=0),WindAmplitude_31_#3=(X=0,Y=0,Z=0),WindOffset_32_#4=(X=0,Y=0,Z=0),GravityOverride_15_#5=(X=0,Y=-350,Z=-980),JointName_27_#6="root")
    GravityOverrideParams_Riding_Moving_55_#21:Gravity Override Setting Struct Structure=(WindFrequency_28_#2=(X=0,Y=0,Z=0),WindAmplitude_31_#3=(X=0,Y=0,Z=0),WindOffset_32_#4=(X=0,Y=0,Z=0),GravityOverride_15_#5=(X=0,Y=-450,Z=-450),JointName_27_#6="root")
    GravityOverrideParams_Riding_Moving_Targeting_58_#22:Gravity Override Setting Struct Structure=(WindFrequency_28_#2=(X=0,Y=0,Z=0),WindAmplitude_31_#3=(X=0,Y=0,Z=0),WindOffset_32_#4=(X=0,Y=0,Z=0),GravityOverride_15_#5=(X=0,Y=-200,Z=-980),JointName_27_#6="root")
    GravityOverrideParams_Tactical_Sprint_57_#23:Gravity Override Setting Struct Structure=(WindFrequency_28_#2=(X=0,Y=0,Z=0),WindAmplitude_31_#3=(X=0,Y=0,Z=0),WindOffset_32_#4=(X=0,Y=0,Z=0),GravityOverride_15_#5=(X=0,Y=-450,Z=-500),JointName_27_#6="root")
    GravityOverrideParams_Rail_Grinding_61_#24:Gravity Override Setting Struct Structure=(WindFrequency_28_#2=(X=0,Y=0,Z=0),WindAmplitude_31_#3=(X=0,Y=0,Z=0),WindOffset_32_#4=(X=0,Y=0,Z=0),GravityOverride_15_#5=(X=0,Y=0,Z=-980),JointName_27_#6="root")
    GravityOverrideParams_DBNO_Tumbling_67_#25:Gravity Override Setting Struct Structure=(WindFrequency_28_#2=(X=0,Y=0,Z=0),WindAmplitude_31_#3=(X=0,Y=0,Z=0),WindOffset_32_#4=(X=0,Y=0,Z=0),GravityOverride_15_#5=(X=-2000,Y=0,Z=0),JointName_27_#6="pelvis")
    GravityOverrideParams_DBNO_RollFwd_68_#26:Gravity Override Setting Struct Structure=(WindFrequency_28_#2=(X=0,Y=0,Z=0),WindAmplitude_31_#3=(X=0,Y=0,Z=0),WindOffset_32_#4=(X=0,Y=0,Z=0),GravityOverride_15_#5=(X=-2000,Y=0,Z=0),JointName_27_#6="pelvis")
    GravityOverrideParams_DBNO_RollBwd_69_#27:Gravity Override Setting Struct Structure=(WindFrequency_28_#2=(X=0,Y=0,Z=0),WindAmplitude_31_#3=(X=0,Y=0,Z=0),WindOffset_32_#4=(X=0,Y=0,Z=0),GravityOverride_15_#5=(X=-4000,Y=0,Z=0),JointName_27_#6="pelvis")
    GravityOverrideParams_DBNO_Impulse_70_#28:Gravity Override Setting Struct Structure=(WindFrequency_28_#2=(X=0,Y=0,Z=0),WindAmplitude_31_#3=(X=0,Y=0,Z=0),WindOffset_32_#4=(X=0,Y=0,Z=0),GravityOverride_15_#5=(X=-2000,Y=0,Z=0),JointName_27_#6="pelvis")
    RBAN Params
    DebugDraw
    DebugJointName
   ) -> (then, GravityOverride, CurrentStateOut, GravityOverrideSettingsOut, RBAN SettingsOut)
   > Set gravOverride(gravOverride:Vector=@n2) -> (Output_Get:Vector)
   n3 = Branch(execute=@n1.then_1, Condition=AND Boolean(A=Get bIsMoving2D(), B=Get bIsCrouching()))
   > Set Crouch walk(Crouch walk:double=1) -> (Output_Get:double)
   Set Crouch walk(execute=@n3.else, Crouch walk=0)
   n4 = GetRealTimeSeconds()
   n5 = "Sin (Radians)"(A="Absolute (Float)"(A=float * float(A=@n4, B=2)))
   Set sineRotation1(execute=@n1.then_2, sineRotation1:Rotator=MakeRotator(Pitch=float * float(A=@n5, B=5))) -> (Output_Get:Rotator)
   > Set sineRotation2(sineRotation2:Rotator=MakeRotator(Pitch=float * float(A=@n5, B=-5))) -> (Output_Get:Rotator)
   n6 = "Sin (Radians)"(A=@n4)
   > Set sineRotation3(sineRotation3:Rotator=MakeRotator(Pitch=float * float(A=@n6, B=25))) -> (Output_Get:Rotator)
   n7 = "Sin (Radians)"(A=float + float(A=@n4, B=-1))
   > Set sineRotation4(sineRotation4:Rotator=MakeRotator(Pitch=float * float(A=@n7, B=27))) -> (Output_Get:Rotator)
   n8 = "Sin (Radians)"(A=float + float(A=@n4, B=-1.5))
   > Set sineRotation5(sineRotation5:Rotator=MakeRotator(Pitch=float * float(A=@n8, B=20))) -> (Output_Get:Rotator)
   n9 = "Sin (Radians)"(A=float + float(A=@n4, B=-2))
   > Set sineRotation6(sineRotation6:Rotator=MakeRotator(Pitch=float * float(A=@n9, B=25))) -> (Output_Get:Rotator)
   n10 = Get bIsMoving2D()
   n11 = Get bIsDBNO()
   Set DBNO_Crawl(execute=@n1.then_3, DBNO_Crawl:double=SelectFloat(A=1, B=0, bPickA=AND Boolean(A=@n10, B=@n11))) -> (Output_Get:double)
   Event BlueprintInitializeAnimation()
   > n12 = Set Skel_Mesh(Skel_Mesh=GetOwnerSkeletalMeshComponent())
  ]}
]
variables [
 Crouch walk {type=number}
 DBNO_Crawl {type=number}
 Gravity Override Params {type=GravityOverrideParamsStruct}
 gravOverride {type=Vector}
 sineRotation1 {type=Rotator}
 sineRotation2 {type=Rotator}
 sineRotation3 {type=Rotator}
 sineRotation4 {type=Rotator}
 sineRotation5 {type=Rotator}
 sineRotation6 {type=Rotator}
 Skel_Mesh {type=SkeletalMeshComponent}
]
