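# Physics asset for the dynamic collar of the Racer Zero mask part on the medium male player character.
# Four collar bodies (upper and bottom, left and right) use large offset capsules, each constrained with swing and twist limits to its tiny MSTR anchor capsule.
# Angularly free constraints tie the MSTR anchors to spine_05, and the head, modeled with three capsules, to spine_05 and each collar body.
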PhysicsAsset /BRCosmetics/Characters/Player/Male/Medium/Bodies/M_MED_Racer_Zero_Mask/Meshes/Parts/M_MED_Racer_Zero_Mask_Dyn
bodies {
 dyn_collar_bott_l [{
   shapeType=Capsule
   length=5.619
   radius=28.0107
   shapeName=dyn_collar_bott_l_capsule
   center=<0 -5.043 26.0473>
   rotation=<rot 90 0.0074 0.0074>
  }]
 dyn_collar_bott_l_MSTR [{shapeType=Capsule, length=0.505, radius=0.505}]
 dyn_collar_bott_r [{
   shapeType=Capsule
   length=5.619
   radius=28.0107
   shapeName=dyn_collar_bott_l_capsule
   center=<0 5.043 26.0473>
   rotation=<rot 90 0.0074 0.0074>
  }]
 dyn_collar_bott_r_MSTR [{shapeType=Capsule, length=0.505, radius=0.505}]
 dyn_collar_l [{
   shapeType=Capsule
   length=13.9097
   radius=16.0903
   shapeName=dyn_collar_l_capsule
   center=<1.2319 -4.9828 14.1202>
   rotation=<rot 0 0 85.4179>
  }]
 dyn_collar_l_MSTR [{shapeType=Capsule, length=0.505, radius=0.505}]
 dyn_collar_r [{
   shapeType=Capsule
   length=13.9097
   radius=16.0903
   shapeName=dyn_collar_l_capsule
   center=<1.2319 4.9828 14.1202>
   rotation=<rot 0 0 85.4179>
  }]
 dyn_collar_r_MSTR [{shapeType=Capsule, length=0.505, radius=0.505}]
 head [
  {shapeType=Capsule, length=10.52, radius=6.9297, center=<2.8448 5.8881 0>, rotation=<rot 90 0 0>}
  {shapeType=Capsule, length=5.9055, radius=1.7588, center=<-3.1169 9.275 -3.8502>, rotation=<rot -34.9997 0 -19.9681>}
  {shapeType=Capsule, length=5.3743, radius=1.7588, center=<-3.2853 8.8962 4.0522>, rotation=<rot -21.3372 -162.1932 144.677>}
 ]
 spine_05 [{shapeType=Capsule, length=5.4275, radius=3.9035, center=<2.0204 3.3222 -0.6332>}]
}
constraints [
 {bone1Name=dyn_collar_r, bone2Name=dyn_collar_r_MSTR, swing1LimitDegrees=20, swing2LimitDegrees=20, twistLimitDegrees=60}
 {bone1Name=dyn_collar_l, bone2Name=dyn_collar_l_MSTR, swing1LimitDegrees=20, swing2LimitDegrees=20, twistLimitDegrees=60}
 {
  bone1Name=head
  bone2Name=spine_05
  swing1LimitDegrees=45
  swing1Motion=Free
  swing2LimitDegrees=45
  swing2Motion=Free
  twistLimitDegrees=45
  twistMotion=Free
 }
 {
  bone1Name=head
  bone2Name=dyn_collar_r
  swing1LimitDegrees=45
  swing1Motion=Free
  swing2LimitDegrees=45
  swing2Motion=Free
  twistLimitDegrees=45
  twistMotion=Free
 }
 {
  bone1Name=head
  bone2Name=dyn_collar_l
  swing1LimitDegrees=45
  swing1Motion=Free
  swing2LimitDegrees=45
  swing2Motion=Free
  twistLimitDegrees=45
  twistMotion=Free
 }
 {bone1Name=dyn_collar_bott_l, bone2Name=dyn_collar_bott_l_MSTR, swing1LimitDegrees=20, swing2LimitDegrees=20, twistLimitDegrees=60}
 {bone1Name=dyn_collar_bott_r, bone2Name=dyn_collar_bott_r_MSTR, swing1LimitDegrees=20, swing2LimitDegrees=20, twistLimitDegrees=60}
 {
  bone1Name=dyn_collar_bott_r
  bone2Name=head
  swing1LimitDegrees=45
  swing1Motion=Free
  swing2LimitDegrees=45
  swing2Motion=Free
  twistLimitDegrees=45
  twistMotion=Free
 }
 {
  bone1Name=dyn_collar_bott_l
  bone2Name=head
  swing1LimitDegrees=45
  swing1Motion=Free
  swing2LimitDegrees=45
  swing2Motion=Free
  twistLimitDegrees=45
  twistMotion=Free
 }
 {
  bone1Name=dyn_collar_bott_l_MSTR
  bone2Name=spine_05
  swing1LimitDegrees=45
  swing1Motion=Free
  swing2LimitDegrees=45
  swing2Motion=Free
  twistLimitDegrees=45
  twistMotion=Free
 }
 {
  bone1Name=dyn_collar_l_MSTR
  bone2Name=spine_05
  swing1LimitDegrees=45
  swing1Motion=Free
  swing2LimitDegrees=45
  swing2Motion=Free
  twistLimitDegrees=45
  twistMotion=Free
 }
 {
  bone1Name=dyn_collar_bott_r_MSTR
  bone2Name=spine_05
  swing1LimitDegrees=45
  swing1Motion=Free
  swing2LimitDegrees=45
  swing2Motion=Free
  twistLimitDegrees=45
  twistMotion=Free
 }
 {
  bone1Name=dyn_collar_r_MSTR
  bone2Name=spine_05
  swing1LimitDegrees=45
  swing1Motion=Free
  swing2LimitDegrees=45
  swing2Motion=Free
  twistLimitDegrees=45
  twistMotion=Free
 }
]
