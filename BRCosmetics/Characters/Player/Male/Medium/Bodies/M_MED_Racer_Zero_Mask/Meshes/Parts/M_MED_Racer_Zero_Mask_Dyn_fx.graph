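# Physics asset for the Racer Zero male medium mask, covering four head trail FX chains (top and bottom, left and right) as dynamic bones.
# A large head capsule roots left and right collar bodies, each feeding top and bottom master bodies that lead into two-segment trail chains.
# Collar and master joints are angularly free with tiny placeholder-sized capsules, while trail segments use elongated capsules with wide swing and tight twist limits.
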PhysicsAsset /BRCosmetics/Characters/Player/Male/Medium/Bodies/M_MED_Racer_Zero_Mask/Meshes/Parts/M_MED_Racer_Zero_Mask_Dyn_fx
bodies {
 dyn_collar_l [{shapeType=Capsule, length=0.5, radius=0.5}]
 dyn_collar_r [{shapeType=Capsule, length=0.5, radius=0.5}]
 dyn_fx_head_trail_bott_l_01 [{shapeType=Capsule, length=4.3175, radius=0.7237, center=<2.4026 -0.0001 0>, rotation=<rot 90 0 0>}]
 dyn_fx_head_trail_bott_l_02 [{shapeType=Capsule, length=4.3175, radius=0.7237, center=<2.4026 -0.0001 0>, rotation=<rot 90 0 0>}]
 dyn_fx_head_trail_bott_l_MSTR [{shapeType=Capsule, length=0.5, radius=0.5}]
 dyn_fx_head_trail_bott_r_01 [{shapeType=Capsule, length=4.3175, radius=0.7237, center=<2.4026 -0.0001 0>, rotation=<rot 90 0 0>}]
 dyn_fx_head_trail_bott_r_02 [{shapeType=Capsule, length=4.3175, radius=0.7237, center=<2.4026 -0.0001 0>, rotation=<rot 90 0 0>}]
 dyn_fx_head_trail_bott_r_MSTR [{shapeType=Capsule, length=0.5, radius=0.5}]
 dyn_fx_head_trail_top_l_01 [{shapeType=Capsule, length=4.3175, radius=0.7237, center=<2.4026 -0.0001 0>, rotation=<rot 90 0 0>}]
 dyn_fx_head_trail_top_l_02 [{shapeType=Capsule, length=4.3175, radius=0.7237, center=<2.4026 -0.0001 0>, rotation=<rot 90 0 0>}]
 dyn_fx_head_trail_top_l_MSTR [{shapeType=Capsule, length=0.5, radius=0.5}]
 dyn_fx_head_trail_top_r_01 [{shapeType=Capsule, length=4.3175, radius=0.7237, center=<2.4026 -0.0001 0>, rotation=<rot 90 0 0>}]
 dyn_fx_head_trail_top_r_02 [{shapeType=Capsule, length=4.3175, radius=0.7237, center=<2.4026 -0.0001 0>, rotation=<rot 90 0 0>}]
 dyn_fx_head_trail_top_r_MSTR [{shapeType=Capsule, length=0.5, radius=0.5}]
 head [{shapeType=Capsule, length=13.0724, radius=10.0818, center=<0.3816 4.7563 0.5325>, rotation=<rot 90 53.9726 63.9726>}]
}
constraints [
 {
  bone1Name=dyn_fx_head_trail_top_r_02
  bone2Name=dyn_fx_head_trail_top_r_01
  swing1LimitDegrees=60
  swing2LimitDegrees=60
  twistLimitDegrees=5
 }
 {
  bone1Name=dyn_collar_r
  bone2Name=head
  swing1LimitDegrees=45
  swing1Motion=Free
  swing2LimitDegrees=45
  swing2Motion=Free
  twistLimitDegrees=45
  twistMotion=Free
 }
 {
  bone1Name=dyn_fx_head_trail_bott_r_MSTR
  bone2Name=dyn_collar_r
  swing1LimitDegrees=45
  swing1Motion=Free
  swing2LimitDegrees=45
  swing2Motion=Free
  twistLimitDegrees=45
  twistMotion=Free
 }
 {
  bone1Name=dyn_fx_head_trail_bott_r_01
  bone2Name=dyn_fx_head_trail_bott_r_MSTR
  swing1LimitDegrees=60
  swing2LimitDegrees=60
  twistLimitDegrees=5
 }
 {
  bone1Name=dyn_fx_head_trail_bott_r_02
  bone2Name=dyn_fx_head_trail_bott_r_01
  swing1LimitDegrees=60
  swing2LimitDegrees=60
  twistLimitDegrees=5
 }
 {
  bone1Name=dyn_fx_head_trail_top_r_MSTR
  bone2Name=dyn_collar_r
  swing1LimitDegrees=45
  swing1Motion=Free
  swing2LimitDegrees=45
  swing2Motion=Free
  twistLimitDegrees=45
  twistMotion=Free
 }
 {
  bone1Name=dyn_fx_head_trail_top_r_01
  bone2Name=dyn_fx_head_trail_top_r_MSTR
  swing1LimitDegrees=60
  swing2LimitDegrees=60
  twistLimitDegrees=5
 }
 {
  bone1Name=dyn_fx_head_trail_top_l_02
  bone2Name=dyn_fx_head_trail_top_l_01
  swing1LimitDegrees=60
  swing2LimitDegrees=60
  twistLimitDegrees=5
 }
 {
  bone1Name=dyn_collar_l
  bone2Name=head
  swing1LimitDegrees=45
  swing1Motion=Free
  swing2LimitDegrees=45
  swing2Motion=Free
  twistLimitDegrees=45
  twistMotion=Free
 }
 {
  bone1Name=dyn_fx_head_trail_bott_l_MSTR
  bone2Name=dyn_collar_l
  swing1LimitDegrees=45
  swing1Motion=Free
  swing2LimitDegrees=45
  swing2Motion=Free
  twistLimitDegrees=45
  twistMotion=Free
 }
 {
  bone1Name=dyn_fx_head_trail_bott_l_01
  bone2Name=dyn_fx_head_trail_bott_l_MSTR
  swing1LimitDegrees=60
  swing2LimitDegrees=60
  twistLimitDegrees=5
 }
 {
  bone1Name=dyn_fx_head_trail_bott_l_02
  bone2Name=dyn_fx_head_trail_bott_l_01
  swing1LimitDegrees=60
  swing2LimitDegrees=60
  twistLimitDegrees=5
 }
 {
  bone1Name=dyn_fx_head_trail_top_l_MSTR
  bone2Name=dyn_collar_l
  swing1LimitDegrees=45
  swing1Motion=Free
  swing2LimitDegrees=45
  swing2Motion=Free
  twistLimitDegrees=45
  twistMotion=Free
 }
 {
  bone1Name=dyn_fx_head_trail_top_l_01
  bone2Name=dyn_fx_head_trail_top_l_MSTR
  swing1LimitDegrees=60
  swing2LimitDegrees=60
  twistLimitDegrees=5
 }
]
