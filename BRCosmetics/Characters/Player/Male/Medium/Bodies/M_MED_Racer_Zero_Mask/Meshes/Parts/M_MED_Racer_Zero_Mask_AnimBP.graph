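# Post-process style AnimBP for the male Racer Zero Mask part, adding secondary motion to its collar and head-trail FX bones.
# Copies the attached parent's pose, simulates a RigidBody pass, then weight-constrains trail master bones to collar, head and torso bones.
# A second RigidBody pass simulates FX under weak upward gravity.
# Upper-arm yaw relative to spine_05 drives additive corrective transforms on the lower collar bones.
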AnimBlueprint /BRCosmetics/Characters/Player/Male/Medium/Bodies/M_MED_Racer_Zero_Mask/Meshes/Parts/M_MED_Racer_Zero_Mask_AnimBP
target_skeleton=&/M_MED_Racer_Zero_Mask_Skeleton
graphs [
 AnimGraph {nodes [
   n1 = RigidBody(ComponentPose=Local To Component(LocalPose=Copy Pose From Mesh() {bUseAttachedParent=true})) {
    bClampLinearTranslationLimitToRefPose=true
    bOverrideWorldGravity=true
    overridePhysicsAsset=&/M_MED_Racer_Zero_Mask_Dyn
    simulationSpace=BaseBoneSpace
    alphaBoolBlend {blendOption=Linear}
    baseBoneRef {boneName=neck_02}
    componentAppliedLinearAccClamp=<3000 3000 3000>
    componentLinearAccScale=<0.4 0.4 0.4>
    componentLinearVelScale=<0.4 0.4 0.4>
    overrideWorldGravity {z=-980}
   }
   n2 = "Constraint - dyn_fx_head_trail_bott_l_MSTR (PPPPP)"!(
    ComponentPose=@n1
    Alpha=1
    ConstraintWeights_0=0.3579
    ConstraintWeights_1:float=0.0887
    ConstraintWeights_2:float=0.0145
    ConstraintWeights_3:float=0.0173
    ConstraintWeights_4:float=0.5217
   ) -> (Pose) {
    alphaBoolBlend {blendOption=Linear}
    boneToModify {boneName=dyn_fx_head_trail_bott_l_MSTR}
    constraintSetup [
     {targetBone {boneName=dyn_collar_l}}
     {targetBone {boneName=head}}
     {targetBone {boneName=spine_05}}
     {targetBone {boneName=clavicle_l}}
     {targetBone {boneName=dyn_collar_bott_l}}
    ]
    constraintWeights [1,1,1,1,1]
   }
   n3 = "Constraint - dyn_fx_head_trail_top_l_MSTR (PPP)"!(ComponentPose=@n2, Alpha=1, ConstraintWeights_0=0.3597, ConstraintWeights_1:float=0.5288, ConstraintWeights_2:float=0.1116) -> (Pose) {
    alphaBoolBlend {blendOption=Linear}
    boneToModify {boneName=dyn_fx_head_trail_top_l_MSTR}
    constraintSetup [{targetBone {boneName=head}}, {targetBone {boneName=dyn_collar_l}}, {targetBone {boneName=dyn_collar_bott_l}}]
    constraintWeights [1,1,1]
   }
   n4 = "Constraint - dyn_fx_head_trail_top_r_MSTR (PPP)"!(ComponentPose=@n3, Alpha=1, ConstraintWeights_0=0.5289, ConstraintWeights_1:float=0.1196, ConstraintWeights_2:float=0.3514) -> (Pose) {
    alphaBoolBlend {blendOption=Linear}
    boneToModify {boneName=dyn_fx_head_trail_top_r_MSTR}
    constraintSetup [{targetBone {boneName=dyn_collar_r}}, {targetBone {boneName=dyn_collar_bott_r}}, {targetBone {boneName=head}}]
    constraintWeights [1,1,1]
   }
   n5 = "Constraint - dyn_fx_head_trail_bott_r_MSTR (PPPPP)"!(
    ComponentPose=@n4
    Alpha=1
    ConstraintWeights_0=0.0163
    ConstraintWeights_1:float=0.3682
    ConstraintWeights_2:float=0.0926
    ConstraintWeights_3:float=0.0137
    ConstraintWeights_4:float=0.5092
   ) -> (Pose) {
    alphaBoolBlend {blendOption=Linear}
    boneToModify {boneName=dyn_fx_head_trail_bott_r_MSTR}
    constraintSetup [
     {targetBone {boneName=clavicle_r}}
     {targetBone {boneName=dyn_collar_r}}
     {targetBone {boneName=head}}
     {targetBone {boneName=spine_05}}
     {targetBone {boneName=dyn_collar_bott_r}}
    ]
    constraintWeights [1,1,1,1,1]
   }
   n6 = RigidBody(ComponentPose=@n5) {
    bClampLinearTranslationLimitToRefPose=true
    bOverrideWorldGravity=true
    overridePhysicsAsset=&/M_MED_Racer_Zero_Mask_Dyn_fx
    simulationSpace=BaseBoneSpace
    alphaBoolBlend {blendOption=Linear}
    baseBoneRef {boneName=neck_02}
    componentAppliedLinearAccClamp=<3000 3000 3000>
    componentLinearAccScale=<0.1 0.2 0.2>
    componentLinearVelScale=<0.1 0.2 0.2>
    overrideWorldGravity {z=100}
   }
   n7 = "Transform (Modify) Bone - Bone: dyn_collar_bott_l"(ComponentPose=@n6, Alpha=Get arm_up_l() -> (arm_up_l:double), Translation="1,1,-2", Rotation="17,-11,-30", Scale="1.000,0.6,1.000") {
    rotationMode=BMM_Additive
    rotationSpace=BCS_BoneSpace
    scaleMode=BMM_Additive
    scaleSpace=BCS_BoneSpace
    translationMode=BMM_Additive
    translationSpace=BCS_BoneSpace
    alphaBoolBlend {blendOption=Linear}
    boneToModify {boneName=dyn_collar_bott_l}
    rotation=<rot 17 -11 -30>
    scale {y=0.6}
    translation=<1 1 -2>
   }
   n8 = "Transform (Modify) Bone - Bone: dyn_collar_bott_r"(ComponentPose=@n7, Alpha=Get arm_up_r() -> (arm_up_r:double), Translation="0,0,-2", Rotation="9,5,30", Scale="1.000,0.7,0.6") {
    rotationMode=BMM_Additive
    rotationSpace=BCS_BoneSpace
    scaleMode=BMM_Additive
    scaleSpace=BCS_BoneSpace
    translationMode=BMM_Additive
    translationSpace=BCS_BoneSpace
    alphaBoolBlend {blendOption=Linear}
    boneToModify {boneName=dyn_collar_bott_r}
    rotation=<rot 9 5 30>
    scale {y=0.7, z=0.6}
    translation {z=-2}
   }
   Output Pose(Result=Component To Local(ComponentPose=@n8)) {layerGroup=None}
  ]}
 EventGraph {nodes [
   Event BlueprintUpdateAnimation()
   > n1 = Sequence()
   n2 = BreakTransform(InTransform=GetDeltaTransformFromRefPose(self=GetOwnerSkeletalMeshComponent(), BoneName=upperarm_r, BaseName=spine_05))
   > Set arm_up_r(arm_up_r:double=MapRangeClamped(Value=BreakRotator(InRot=@n2.Rotation).Yaw, InRangeA=34, InRangeB=100)) -> (Output_Get:double)
   n3 = BreakTransform(InTransform=GetDeltaTransformFromRefPose(self=GetOwnerSkeletalMeshComponent(), BoneName=upperarm_l, BaseName=spine_05))
   Set arm_up_l(execute=@n1.then_1, arm_up_l:double=MapRangeClamped(Value=BreakRotator(InRot=@n3.Rotation).Yaw, InRangeA=25, InRangeB=60)) -> (Output_Get:double)
  ]}
]
variables [
 arm_up_l {type=number}
 arm_up_r {type=number}
]
